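; BP_UIDirectorComponent_LTM_HighTower is a UI director component for the HighTower limited-time mode that manages its HUD scenes.
; When the UI is ready to show, it adds the scenes from GetDefaultHUDScenes.
; On EndPlay, it removes the same scenes.
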(event EventOnReadyToShowUI
  (CallFunction|AddScenes (Variables|Default|GetDefaultHUDScenes)))

(event EventEndPlay
  (params (EndPlayReason :type EEndPlayReason%20Enum :default "Destroyed"))
  (CallFunction|RemoveScenes (Variables|Default|GetDefaultHUDScenes)))
