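# HUD widget for one HighTower ability socket, keyed by ActionTag, showing icon, keybind, cooldown and charge.
# Tag maps select socket, charge and pulse art; inventory events matching the tag async-load the icon into a dynamic material and animate socketing.
# A collapsed QuickbarSlotCooldown supplies cooldown events that drive material percent, countdown text and start/finish animations.
# Interact events pulse and glow the socket for matching pickups; touch switches to the primary quickbar.
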WidgetBlueprint /HighTower/UI/HUD_AbilityWidget
animations [
 OffCooldown {
  binding_count=10
  length_seconds=1.8179
  object=&self:OffCooldown
  sequence {
   bindings [
    Image_Glow {
     id=#1
     possessed_class=Image
     tracks [
      {
       type=MovieScene2DTransformTrack
       sections [{
         type=MovieScene2DTransformSection
         channels [
          Translation.X {type=MovieSceneScriptingFloatChannel}
          Translation.Y {type=MovieSceneScriptingFloatChannel}
          Angle {type=MovieSceneScriptingFloatChannel}
          Scale.X {type=MovieSceneScriptingFloatChannel, keys [[-0.073,1],[12,1.1552],[55.73,0]]}
          Scale.Y {type=MovieSceneScriptingFloatChannel, keys [[-0.073,1],[12,1.1552],[55.73,0]]}
          Shear.X {type=MovieSceneScriptingFloatChannel}
          Shear.Y {type=MovieSceneScriptingFloatChannel}
         ]
        }]
      }
      {
       type=MovieSceneFloatTrack
       display_name=Render Opacity
       sections [{type=MovieSceneFloatSection, channels [None {type=MovieSceneScriptingFloatChannel, keys [[0,0],[6,1]]}]}]
      }
     ]
    }
    CommonBorder_Input {
     id=#2
     possessed_class=CommonBorder
     tracks [{
       type=MovieSceneColorTrack
       display_name=Content Color and Opacity
       sections [{
         type=MovieSceneColorSection
         channels [
          Color.R {type=MovieSceneScriptingFloatChannel, keys [[0,0.0807],[2.723,1]]}
          Color.G {type=MovieSceneScriptingFloatChannel, keys [[0,0.0802],[2.723,1]]}
          Color.B {type=MovieSceneScriptingFloatChannel, keys [[0,0.0798],[2.723,1]]}
          Color.A {type=MovieSceneScriptingFloatChannel, const [1,2]}
         ]
        }]
      }]
    }
    Image_Spark1_2 {
     id=#3
     possessed_class=Image
     tracks [{
       type=MovieSceneFloatTrack
       display_name=Render Opacity
       sections [{type=MovieSceneFloatSection, channels [None {type=MovieSceneScriptingFloatChannel, keys [[20.783,0],[32.755,1],[42.118,0]]}]}]
      }]
    }
    Number_Cooldown {
     id=#4
     display_name="Image_Cooldownbg (9)"
     possessed_class=CommonNumericTextBlock
     tracks [{
       type=MovieSceneFloatTrack
       display_name=Render Opacity
       sections [{type=MovieSceneFloatSection, channels [None {type=MovieSceneScriptingFloatChannel, keys [[0,1],[5,0]]}]}]
      }]
    }
    Image_Spark1_4 {
     id=#5
     possessed_class=Image
     tracks [{
       type=MovieSceneFloatTrack
       display_name=Render Opacity
       sections [{type=MovieSceneFloatSection, channels [None {type=MovieSceneScriptingFloatChannel, keys [[25.863,0],[37.835,1],[48.118,0]]}]}]
      }]
    }
    Image_Spark1 {
     id=#6
     possessed_class=Image
     tracks [{
       type=MovieSceneFloatTrack
       display_name=Render Opacity
       sections [{type=MovieSceneFloatSection, channels [None {type=MovieSceneScriptingFloatChannel, keys [[9.237,0],[21.209,1],[32.483,0]]}]}]
      }]
    }
    Image_Spark1_3 {
     id=#7
     possessed_class=Image
     tracks [{
       type=MovieSceneFloatTrack
       display_name=Render Opacity
       sections [{type=MovieSceneFloatSection, channels [None {type=MovieSceneScriptingFloatChannel, keys [[15.548,0],[27.52,1],[37.835,0]]}]}]
      }]
    }
    Image_Cooldownbg {
     id=#8
     possessed_class=Image
     tracks [
      {
       type=MovieSceneFloatTrack
       display_name=Render Opacity
       sections [{type=MovieSceneFloatSection, channels [None {type=MovieSceneScriptingFloatChannel, keys [[0,0],[10.704,1]]}]}]
      }
      {
       type=MovieSceneWidgetMaterialTrack
       sections [{
         type=MovieSceneParameterSection
         channels [
          "Density (S)" {type=MovieSceneScriptingFloatChannel, keys [[0,0],[10.704,3.0377],[55.73,0]]}
          "Density2 (S)" {type=MovieSceneScriptingFloatChannel, keys [[0,0],[10.704,0.225],[55.73,0]]}
          "Density3 (S)" {type=MovieSceneScriptingFloatChannel, keys [[0,0],[10.704,5.9728],[55.73,0]]}
         ]
        }]
      }
      {
       type=MovieScene2DTransformTrack
       sections [{
         type=MovieScene2DTransformSection
         channels [
          Translation.X {type=MovieSceneScriptingFloatChannel}
          Translation.Y {type=MovieSceneScriptingFloatChannel}
          Angle {type=MovieSceneScriptingFloatChannel}
          Scale.X {type=MovieSceneScriptingFloatChannel, keys [[-0.073,1],[10.704,1.4408],[27,1]]}
          Scale.Y {type=MovieSceneScriptingFloatChannel, keys [[-0.073,1],[10.704,1.0481],[27,1]]}
          Shear.X {type=MovieSceneScriptingFloatChannel}
          Shear.Y {type=MovieSceneScriptingFloatChannel}
         ]
        }]
      }
     ]
    }
    Image_AbilityIcon {
     id=#9
     possessed_class=Image
     tracks [{
       type=MovieSceneWidgetMaterialTrack
       sections [{
         type=MovieSceneParameterSection
         channels [
          CooldownSwitch {type=MovieSceneScriptingFloatChannel, keys [[-0.073,1],[6,0]]}
          ShineOffset {type=MovieSceneScriptingFloatChannel, keys [[0,0.4314],[109,1.5357]]}
          Whiteout {type=MovieSceneScriptingFloatChannel, keys [[-0.073,0],[6,1],[10.704,0]]}
         ]
        }]
      }]
    }
    Image_Spark1_1 {
     id=#10
     possessed_class=Image
     tracks [{
       type=MovieSceneFloatTrack
       display_name=Render Opacity
       sections [{type=MovieSceneFloatSection, channels [None {type=MovieSceneScriptingFloatChannel, keys [[12.623,0],[24.595,1],[34.618,0]]}]}]
      }]
    }
   ]
   display_rate {numerator=60}
  }
 }
 StartCooldown {
  binding_count=3
  length_seconds=0.35
  object=&self:StartCooldown
  sequence {bindings [
    CommonBorder_Input {
     id=#11
     possessed_class=CommonBorder
     tracks [{
       type=MovieSceneColorTrack
       display_name=Content Color and Opacity
       sections [{
         type=MovieSceneColorSection
         channels [
          Color.R {type=MovieSceneScriptingFloatChannel, keys [[0,1],[2.0003,0.0807]]}
          Color.G {type=MovieSceneScriptingFloatChannel, keys [[0,1],[2.0003,0.0802]]}
          Color.B {type=MovieSceneScriptingFloatChannel, keys [[0,1],[2.0003,0.0798]]}
          Color.A {type=MovieSceneScriptingFloatChannel, const [1,2]}
         ]
        }]
      }]
    }
    Image_AbilityIcon {
     id=#12
     possessed_class=Image
     tracks [{
       type=MovieSceneWidgetMaterialTrack
       sections [{type=MovieSceneParameterSection, channels [CooldownSwitch {type=MovieSceneScriptingFloatChannel, keys [[0,0],[2.0003,1]]}]}]
      }]
    }
    Number_Cooldown {
     id=#13
     display_name="Image_AbilityIcon (3)"
     possessed_class=CommonNumericTextBlock
     tracks [
      {
       type=MovieSceneFloatTrack
       display_name=Render Opacity
       sections [{type=MovieSceneFloatSection, channels [None {type=MovieSceneScriptingFloatChannel, keys [[0,0],[7,1]]}]}]
      }
      {
       type=MovieScene2DTransformTrack
       sections [{
         type=MovieScene2DTransformSection
         channels [
          Translation.X {type=MovieSceneScriptingFloatChannel}
          Translation.Y {type=MovieSceneScriptingFloatChannel}
          Angle {type=MovieSceneScriptingFloatChannel}
          Scale.X {type=MovieSceneScriptingFloatChannel, keys [[0,2],[7,1]]}
          Scale.Y {type=MovieSceneScriptingFloatChannel, keys [[0,2],[7,1]]}
          Shear.X {type=MovieSceneScriptingFloatChannel}
          Shear.Y {type=MovieSceneScriptingFloatChannel}
         ]
        }]
      }
     ]
    }
   ]}
 }
 Socket {
  binding_count=1
  length_seconds=0.4
  object=&self:Socket
  sequence {bindings [Image_AbilityIcon {
     id=#14
     possessed_class=Image
     tracks [{
       type=MovieSceneWidgetMaterialTrack
       sections [{type=MovieSceneParameterSection, channels [UseIcon {type=MovieSceneScriptingFloatChannel, keys [[0,0],[8,1]]}]}]
      }]
    }]}
 }
 UnSocket {binding_count=0, length_seconds=5, object=&self:UnSocket, sequence {bindings []}}
 Activate {
  binding_count=1
  length_seconds=0.2157
  object=&self:Activate
  sequence {bindings [Image_AbilityIcon {
     id=#15
     possessed_class=Image
     tracks [
      {
       type=MovieSceneWidgetMaterialTrack
       sections [{type=MovieSceneParameterSection, channels [Whiteout {type=MovieSceneScriptingFloatChannel, keys [[0,0],[0.7097,1],[4,0]]}]}]
      }
      {
       type=MovieScene2DTransformTrack
       sections [{
         type=MovieScene2DTransformSection
         channels [
          Translation.X {type=MovieSceneScriptingFloatChannel}
          Translation.Y {type=MovieSceneScriptingFloatChannel}
          Angle {type=MovieSceneScriptingFloatChannel}
          Scale.X {type=MovieSceneScriptingFloatChannel, keys [[0,1],[2,0.8752],[3.3083,1.0338],[4.3143,1.0006]]}
          Scale.Y {type=MovieSceneScriptingFloatChannel, keys [[0,1],[2,0.8752],[3.3083,1.0338],[4.3143,1.0006]]}
          Shear.X {type=MovieSceneScriptingFloatChannel}
          Shear.Y {type=MovieSceneScriptingFloatChannel}
         ]
        }]
      }
     ]
    }]}
 }
 StartCharge {
  binding_count=1
  length_seconds=0.3426
  object=&self:StartCharge
  sequence {bindings [Image_Charge {
     id=#16
     possessed_class=Image
     tracks [{
       type=MovieSceneFloatTrack
       display_name=Render Opacity
       sections [{type=MovieSceneFloatSection, channels [None {type=MovieSceneScriptingFloatChannel, keys [[0,0],[6.134,1]]}]}]
      }]
    }]}
 }
 ToggleGlow {
  binding_count=1
  length_seconds=0.2318
  object=&self:ToggleGlow
  sequence {bindings [Image_AbilityIcon {
     id=#17
     possessed_class=Image
     tracks [{
       type=MovieSceneWidgetMaterialTrack
       sections [{type=MovieSceneParameterSection, channels [GlowToggle {type=MovieSceneScriptingFloatChannel, keys [[0,0],[4.635,1]]}]}]
      }]
    }]}
 }
]
graphs [
 EventGraph {nodes [
   "OnCooldownStarted (QuickbarSlotCooldown)"!() -> (OutputDelegate, then, CooldownType:EFortItemCooldownType Enum=None, CooldownDuration:float) {is=title K2Node_ComponentBoundEvent}
   > n1 = Branch(Condition=IsAnimationPlaying(InAnimation=Get OffCooldown() -> (OffCooldown:Widget Animation Object Reference)))
   > n2 = StopAnimation(InAnimation=Get OffCooldown())
   PlayAnimation(execute=[@n2.then, @n1.else], InAnimation=Get StartCooldown() -> (StartCooldown:Widget Animation Object Reference))
   "OnCooldownStopped (QuickbarSlotCooldown)"!() -> (OutputDelegate, then, CooldownType:EFortItemCooldownType Enum=None) {is=title K2Node_ComponentBoundEvent}
   n3 = Get OffCooldown()
   > n4 = Branch(Condition=IsAnimationPlaying(InAnimation=@n3))
   PlayAnimation(execute=@n4.else, InAnimation=@n3)
   n5 = "OnCooldownUpdated (QuickbarSlotCooldown)"!() -> (
    OutputDelegate
    then
    CooldownType:EFortItemCooldownType Enum=None
    CooldownDuration:float
    CooldownRemaining:float
    CooldownPercentage:float
   ) {is=title K2Node_ComponentBoundEvent}
   > EventSetCooldown!(execute, self, Remaining:double=@n5.CooldownRemaining, Percent:double=@n5.CooldownPercentage) -> (then)
   n6 = Get OffCooldown()
   n7 = IsAnimationPlaying(InAnimation=@n6)
   > Branch(Condition=AND Boolean(A="float <= float"!(A:double=@n5.CooldownRemaining, B:double=0) -> (ReturnValue:bool) {is=title K2Node_CallFunction}, B=NOT Boolean(A=@n7)))
   > PlayAnimation(InAnimation=@n6)
   n8 = OnItemAdded_Bind!() -> (OutputDelegate, then, Item:Fort Item Object Reference, InventoryOwner:Fort Inventory Owner Interface Interface) {is=title K2Node_CustomEvent}
   n9 = GetOwningPlayer()
   > n10 = Cast To FortPlayerControllerSpectating!(execute, Object=@n9) -> (then, CastFailed, AsFort Player Controller Spectating:Fort Player Controller Spectating Object Reference) {is=title K2Node_DynamicCast}
   > PrintString!(execute, InString="\r\n--------------\r\nSpectator\r\n--------------\r\n", bPrintToScreen, bPrintToLog, TextColor, Duration, Key) -> (then)
   Branch(
    execute=@n10.CastFailed
    Condition="Equal (Object)"(A="To Object (Interface)"!(Interface:"Interface Interface (by ref)"=@n8.InventoryOwner) -> (ReturnValue:Object Reference), B=@n9) {is=title K2Node_CallFunction}
   )
   n11 = Get FortInventorySubsystem() -> (ReturnValue:Fort Inventory Subsystem Object Reference) {is=title K2Node_GetSubsystem}
   > Branch(Condition="Equal (GameplayTag)"!(
     A:Gameplay Tag Structure=GetActionTagForItem!(self:Fort Inventory Subsystem Object Reference=@n11, Item:Fort Item Object Reference=@n8.Item) -> (ReturnValue:Gameplay Tag Structure)
     B:Gameplay Tag Structure=Get ActionTag() -> (ActionTag:Gameplay Tag Structure)
    ) -> (ReturnValue:bool) {is=title K2Node_CallFunction})
   > n12 = Async Load Asset(Asset=GetSmallPreviewImage!(self:Fort Item Object Reference=@n8.Item) -> (ReturnValue:Texture 2D Soft Object Reference))
   > n13 = Cast To FortWorldItem!(execute, Object=@n8.Item) -> (then, CastFailed, AsFort World Item:Fort World Item Object Reference) {is=title K2Node_DynamicCast}
   > SetItem!(
    execute
    self:Fort Item Cooldown Widget Object Reference=Get QuickbarSlotCooldown() -> (QuickbarSlotCooldown:Quickbar Slot Cooldown Object Reference)
    InWorldItem:Fort World Item Object Reference=@n13
   ) -> (then)
   > EventSocket!(execute, self, isFull:bool=true) -> (then)
   n14 = Cast To Texture!(execute=@n12.Completed, Object=@n12) -> (then, CastFailed, AsTexture:Texture Object Reference) {is=title K2Node_DynamicCast}
   > SetTextureParameterValue(
    self=Get AbilityIconMID() -> (AbilityIconMID:Material Instance Dynamic Object Reference)
    ParameterName=Get Material Param Name AbilityIcon Const() -> (Material Param Name AbilityIcon Const:Name=None)
    Value=@n14
   )
   EventSocket(execute=@n14.CastFailed, isFull=false)
   n15 = EventSetCooldown!() -> (OutputDelegate, then, Remaining:double=~, Percent:double=~) {is=title K2Node_CustomEvent}
   > SetScalarParameterValue(self=Get AbilityIconMID(), ParameterName=CooldownPercent, Value=@n15.Percent)
   n16 = "float > float"(A=@n15.Remaining, B=1) {is=title K2Node_CallFunction}
   n17 = Get Text_Cooldown() -> (Text_Cooldown:Common Text Object Reference)
   > "SetText (Text)"(
    self=@n17
    InText="To Text (Float)"(
     Value="Clamp (Float)"(Value=@n15.Remaining, Min=0, Max=10000)
     bUseGrouping=false
     MinimumIntegralDigits=1
     MaximumIntegralDigits=324
     MinimumFractionalDigits=0
     MaximumFractionalDigits=Select(Option 0:int=1, Option 1:int=0, Index=@n16) -> (ReturnValue:int)
    )
   )
   > SetRenderTransform!(
    execute
    self:Widget Object Reference=@n17
    InTransform:Widget Transform Structure=Make Widget Transform!(
     Translation:Vector 2D Structure=(X=0,Y=0)
     Scale:Vector 2D Structure=Select(Option 0:Vector 2D Structure=(X=0.7,Y=0.7), Option 1:Vector 2D Structure=(X=1,Y=1), Index=@n16) -> (ReturnValue:Vector 2D Structure)
     Shear:Vector 2D Structure=(X=-8,Y=0)
     Angle:float=0
    ) -> (WidgetTransform:Widget Transform Structure) {is=title K2Node_MakeStruct}
   ) -> (then)
   n18 = OnItemRemoved_Bind!() -> (OutputDelegate, then, Item:Fort Item Object Reference, InventoryOwner:Fort Inventory Owner Interface Interface) {is=title K2Node_CustomEvent}
   > Branch(Condition="Equal (Object)"(A="To Object (Interface)"(Interface=@n18.InventoryOwner), B=GetOwningPlayer()))
   n19 = Get FortInventorySubsystem()
   > Branch(Condition="Equal (GameplayTag)"(A=GetActionTagForItem(self=@n19, Item=@n18.Item), B=Get ActionTag()))
   > SetItem(self=Get QuickbarSlotCooldown())
   > EventSocket(isFull=false)
   EventActivate() {is=title K2Node_CustomEvent}
   > PlayAnimation(InAnimation=Get Activate() -> (Activate:Widget Animation Object Reference))
   n20 = EventSetChargeAmount!() -> (OutputDelegate, then, ChargePercent:double=~) {is=title K2Node_CustomEvent}
   > SetScalarParameterValue(
    self=Get AbilityChargeMID() -> (AbilityChargeMID:Material Instance Dynamic Object Reference)
    ParameterName=ChargePercent
    Value=@n20.ChargePercent
   )
   n21 = EventOnPlayerInteract!() -> (OutputDelegate, then, ContextInfo:Fort Interact Context Info Object Reference) {is=title K2Node_CustomEvent}
   > n22 = Is Valid(InputObject=@n21.ContextInfo)
   Branch(execute=@n22.Is Not Valid, Condition=Get IsInteracted() -> (IsInteracted:bool))
   > n23 = SetVisibility(self=Get Image_InteractPulse() -> (Image_InteractPulse:Image Object Reference), InVisibility=Collapsed)
   Branch(execute=[@n27.else, @n23.then], Condition=Get IsInteracted())
   n24 = Get ToggleGlow() -> (ToggleGlow:Widget Animation Object Reference)
   > PlayAnimationReverse(InAnimation=@n24)
   > Set IsInteracted(IsInteracted:bool=false) -> (Output_Get:bool)
   n25 = Cast To FortPickup!(
    execute=@n22.Is Valid
    Object=Get ReceivingActor!(self:Fort Interact Context Info Object Reference=@n21.ContextInfo) -> (ReceivingActor:Actor Weak object)
   ) -> (then, CastFailed, AsFort Pickup:Fort Pickup Object Reference) {is=title K2Node_DynamicCast}
   n26 = "Equal (GameplayTag)"(
    A=GetActionTagForItem(
     self=Get FortInventorySubsystem()
     Item=Get PrimaryPickupDummyItem!(self:Fort Pickup Object Reference=@n25) -> (PrimaryPickupDummyItem:Fort Item Object Reference)
    )
    B=Get ActionTag()
   )
   > SetVisibility(
    self=Get Image_InteractPulse()
    InVisibility=Select(Option 0:ESlateVisibility Enum=Collapsed, Option 1:ESlateVisibility Enum=HitTestInvisible, Index=@n26) -> (ReturnValue:ESlateVisibility Enum)
   )
   > n27 = Branch(Condition=@n26)
   > n28 = Branch(Condition=Get IsInteracted())
   PlayAnimationForward(execute=@n28.else, InAnimation=@n24)
   > Set IsInteracted(IsInteracted=true)
   n29 = OnViewTargetInventoryUpdated_Bind() {is=title K2Node_CustomEvent}
   n30 = Cast To FortPlayerController(execute=[@n29.then, @n36.then], Object=GetOwningPlayer())
   n31 = GetItemForActionTag!(
    self:Fort Inventory Subsystem Object Reference=Get FortInventorySubsystem()
    OwningPlayerController:Fort Player Controller Object Reference=@n30
    ActionTag:Gameplay Tag Structure=Get ActionTag()
   ) -> (ReturnValue:Fort World Item Object Reference)
   > n32 = Is Valid(InputObject=@n31)
   > n33 = Async Load Asset(Asset=GetSmallPreviewImage(self=@n31))
   > SetItem(self=Get QuickbarSlotCooldown(), InWorldItem=@n31)
   > EventSocket(isFull=true)
   n34 = Cast To Texture(execute=@n33.Completed, Object=@n33)
   n35 = Get Material Param Name AbilityIcon Const() -> (Material Param Name AbilityIcon Const=None)
   > SetTextureParameterValue(self=Get AbilityIconMID(), ParameterName=@n35, Value=@n34)
   SetTextureParameterValue(execute=[@n34.CastFailed, @n32.Is Not Valid], self=Get AbilityIconMID(), ParameterName=@n35)
   > EventSocket(isFull=false)
   > EventSetCooldown(Remaining=0, Percent=1) {is=title K2Node_CallFunction}
   > "SetText (Text)"(self=Get Text_Cooldown())
   n36 = OnRefreshHUDForRespawn_Bind() {is=title K2Node_CustomEvent}
   n37 = EventSocket!() -> (OutputDelegate, then, isFull:bool=~) {is=title K2Node_CustomEvent}
   > n38 = Branch(Condition=@n37.isFull)
   n39 = Get Socket() -> (Socket:Widget Animation Object Reference)
   > PlayAnimation(InAnimation=@n39)
   PlayAnimationReverse(execute=@n38.else, InAnimation=@n39, PlaybackSpeed=3)
   Event Construct()
   > n40 = Sequence!(execute) -> (then_0, then_1, then_2:Exec, then_3:Exec)
   n41 = Find(
    TargetMap:Map of Gameplay Tag Structures to Material Instance Object References=Get TagMaterialMap() -> (TagMaterialMap:Map of Gameplay Tag Structures to Material Instance Object References)
    Key:"Gameplay Tag Structure (by ref)"=Get ActionTag()
   ) -> (Value:Material Instance Object Reference)
   n42 = Get Image_AbilityIcon() -> (Image_AbilityIcon:Image Object Reference)
   SetBrushFromMaterial(
    execute=[@n40.then_0, @n59.then]
    self=@n42
    Material=Select(
     Option 0:Material Interface Object Reference=/HighTower/UI/Materials/M_UI_AbilitySocket_Secondary
     Option 1:Material Interface Object Reference=@n41
     Index=@n41.ReturnValue
    ) -> (ReturnValue:Material Interface Object Reference)
   )
   > n43 = GetDynamicMaterial(self=@n42)
   > Set AbilityIconMID(AbilityIconMID:Material Instance Dynamic Object Reference=@n43) -> (Output_Get:Material Instance Dynamic Object Reference)
   > n44 = GetDynamicMaterial(self=Get Image_Charge() -> (Image_Charge:Image Object Reference))
   > n45 = Set AbilityChargeMID(AbilityChargeMID:Material Instance Dynamic Object Reference=@n44) -> (Output_Get:Material Instance Dynamic Object Reference)
   n46 = Find(
    TargetMap:Map of Gameplay Tag Structures to Texture Object References=Get TagChargeMap() -> (TagChargeMap:Map of Gameplay Tag Structures to Texture Object References)
    Key=Get ActionTag()
   ) -> (Value:Texture Object Reference)
   > SetTextureParameterValue(
    self=@n45
    ParameterName=Texture
    Value=Select(Option 0:Texture Object Reference, Option 1:Texture Object Reference=@n46, Index=@n46.ReturnValue) -> (ReturnValue:Texture Object Reference)
   )
   n47 = Find(
    TargetMap:Map of Gameplay Tag Structures to Material Instance Object References=Get TagPulseMap() -> (TagPulseMap:Map of Gameplay Tag Structures to Material Instance Object References)
    Key=Get ActionTag()
   ) -> (Value:Material Instance Object Reference)
   > SetBrushFromMaterial(
    self=Get Image_InteractPulse()
    Material=Select(
     Option 0:Material Interface Object Reference=/Game/Athena/HUD/LTM/Infiltration/Materials/M_UI_GameplayDiamondPing_Pulse_Gradient_Inst
     Option 1:Material Interface Object Reference=@n47
     Index=@n47.ReturnValue
    ) -> (ReturnValue:Material Interface Object Reference)
   )
   n48 = Get FortInventorySubsystem()
   Bind Event to On Item Added(execute=@n40.then_1, self:Fort Inventory Subsystem Object Reference=@n48, Delegate=Create Event()) {is=title K2Node_AddDelegate}
   > Bind Event to On Item Removed(self:Fort Inventory Subsystem Object Reference=@n48, Delegate=Create Event()) {is=title K2Node_AddDelegate}
   n49 = Create Event()
   n50 = Get FortHUDContext!() -> (ReturnValue:Fort HUDContext Object Reference) {is=title K2Node_GetBlueprintContextEx}
   > Bind Event to On Player Can Interact(self:Fort HUDContext Object Reference=@n50, Delegate=@n49) {is=title K2Node_AddDelegate}
   > Bind Event to On Interact Updated(self:Fort HUDContext Object Reference=@n50, Delegate=@n49) {is=title K2Node_AddDelegate}
   n51 = Get ActionTag()
   n52 = Get PrimaryAbility() -> (PrimaryAbility:bool)
   n53 = Get PrimaryAltAbility() -> (PrimaryAltAbility:bool)
   n54 = Get FortGlobalUIContext() -> (ReturnValue:Fort Global UIContext Object Reference) {is=title K2Node_GetSubsystem}
   n55 = GetFireActionName!(self:Fort Global UIContext Object Reference=@n54) -> (ReturnValue:Name=None)
   n56 = GetTargetActionName!(self:Fort Global UIContext Object Reference=@n54, OverrideInputType:ECommonInputType Enum=Count) -> (ReturnValue:Name=None)
   n57 = Get PrimaryAbility()
   > SetBoundAction!(
    execute
    self:Fort Keybind Widget Object Reference=Get KeybindWidget() -> (KeybindWidget:Keybind Widget Object Reference)
    NewBoundAction:Name=Select(
     Option 0:Name=GetTagName!(GameplayTag:"Gameplay Tag Structure (by ref)"=@n51) -> (ReturnValue:Name=None)
     Option 1:Name=Select(Option 0=@n56, Option 1=@n55, Index=@n57)
     Index=OR Boolean(A=@n52, B=@n53)
    ) -> (ReturnValue:Name)
   ) -> (then)
   n58 = Get QuickbarSlotCooldown()
   > SetItem(self=@n58)
   > Initialize!(
    execute
    self:Quickbar Slot Cooldown Object Reference=@n58
    QuickbarItemDisplayStyle:ItemDisplayStyle Enum=NewEnumerator5
    DesiredCooldownTypesSupported:Array of EFortItemCooldownType Enums=Make Array("[0]":EFortItemCooldownType Enum=AbilitySetActivateByInputAbility) -> (Array:Array of EFortItemCooldownType Enums)
   ) -> (then)
   > Branch(Condition=Get PrimaryAltAbility())
   > SetRenderScale!(execute, self:Widget Object Reference=Get Image_AbilityIcon(), Scale:Vector 2D Structure=(X=-1,Y=1)) -> (then)
   Reroute Node(InputPin=@n40.then_2)
   Bind Event to On View Target Inventory Updated(execute=@n40.then_3, self:Fort Inventory Subsystem Object Reference=Get FortInventorySubsystem(), Delegate=Create Event()) {is=title K2Node_AddDelegate}
   > Assign On Refresh HUDFor Respawn!(execute:Exec, self:Fort Global UIContext Object Reference=Get FortGlobalUIContext(), Delegate:"Delegate (by ref)"=Create Event()) -> (then:Exec) {is=title K2Node_AssignDelegate}
   Event PreConstruct()
   > n59 = Branch(Condition=IsGameplayTagValid!(GameplayTag:Gameplay Tag Structure=Get ActionTag()) -> (ReturnValue:bool))
   n60 = For Each Loop(Array:Array of Fort World Item Object References=Get World Item List!(
     self:Fort Inventory Context Object Reference=Get FortInventoryContext!() -> (ReturnValue:Fort Inventory Context Object Reference) {is=title K2Node_GetBlueprintContextEx}
     FilterSettings:"Fort Item List Filter Structure (by ref)"=Make Fort Item List Filter!(SearchText:String, FilterType:EFortInventoryFilter Enum=Gadget, bInStorageVault:bool=False, bIncludeVaultOverflow:bool=False) -> (FortItemListFilter:Fort Item List Filter Structure) {is=title K2Node_MakeStruct}
    ) -> (Items:Array of Fort World Item Object References) {is=title K2Node_CallFunction}) -> (Array Element:Fort World Item Object Reference)
   > n61 = Cast To FortPlayerController(Object=GetOwningPlayer())
   > OnItemAdded_Bind!(execute, self, Item:Fort Item Object Reference=@n60, InventoryOwner:Fort Inventory Owner Interface Interface=@n61) -> (then) {is=title K2Node_CallFunction}
  ]}
 OnTouchStarted {nodes [
   OnTouchStarted!() -> (then, MyGeometry:Geometry Structure, InTouchEvent:"Pointer Event Structure (by ref)") {is=title K2Node_FunctionEntry}
   > SwitchToPrimaryQuickbar(self:Athena HUDContext Object Reference=Get AthenaHUDContext!() -> (ReturnValue:Athena HUDContext Object Reference) {is=title K2Node_GetBlueprintContextEx})
   > Return Node!(execute, ReturnValue:Event Reply Structure=Handled())
  ]}
]
variables [
 AbilityChargeMID {type=MaterialInstanceDynamic}
 AbilityIconMID {type=MaterialInstanceDynamic}
 ActionTag {type=GameplayTag, default {tagName=None}}
 IsInteracted {type=boolean}
 Material Param Name AbilityIcon Const {type=string, default=AbilityIcon}
 NewVar_0 {type=FortWorldItem, default=null}
 PrimaryAbility {type=boolean}
 PrimaryAltAbility {type=boolean}
 TagChargeMap {default {
   "(TagName=\"Trait.HighTower.Power.Type.Jump\")"=/HighTower/UI/Materials/T_UI_Hightower_Charge_Jump
   "(TagName=\"Trait.HighTower.Power.Type.Primary\")"=/HighTower/UI/Materials/T_UI_Hightower_Charge_Primary
   "(TagName=\"Trait.HighTower.Power.Type.Secondary\")"=/HighTower/UI/Materials/T_UI_Hightower_Charge_Secondary
   "(TagName=\"Trait.HighTower.Power.Type.Utility\")"=/HighTower/UI/Materials/T_UI_Hightower_Charge_Utility
  }}
 TagMaterialMap {default {
   "(TagName=\"Trait.HighTower.Power.Type.Jump\")"=/HighTower/UI/Materials/M_UI_AbilitySocket_Jump
   "(TagName=\"Trait.HighTower.Power.Type.Primary\")"=/HighTower/UI/Materials/M_UI_AbilitySocket_Primary
   "(TagName=\"Trait.HighTower.Power.Type.Secondary\")"=&M_UI_AbilitySocket_Secondary
   "(TagName=\"Trait.HighTower.Power.Type.Utility\")"=/HighTower/UI/Materials/M_UI_AbilitySocket_Utility
  }}
 TagPulseMap {default {
   "(TagName=\"Trait.HighTower.Power.Type.Jump\")"=/HighTower/UI/Materials/M_UI_GameplayDiamondPing_Pulse_Jump
   "(TagName=\"Trait.HighTower.Power.Type.Primary\")"=/HighTower/UI/Materials/M_UI_GameplayDiamondPing_Pulse_Primary
   "(TagName=\"Trait.HighTower.Power.Type.Secondary\")"=/HighTower/UI/Materials/M_UI_GameplayDiamondPing_Pulse_Secondary
   "(TagName=\"Trait.HighTower.Power.Type.Utility\")"=/HighTower/UI/Materials/M_UI_GameplayDiamondPing_Pulse_Utility
  }}
]
widgets {
 info {inheritedWidgetCount=0, parentClass=&UserWidget, rootWidgetClass=&Overlay, widgetCount=20}
 widgets [
  Overlay_13 {widgetClassPath=&Overlay, widget=Overlay_13}
  Overlay_35 {
   widgetClassPath=&Overlay
   parent=Overlay_13
   widget=Overlay_35
   slot {class=OverlaySlot, properties {horizontalAlignment=HAlign_Center, verticalAlignment=VAlign_Center}}
  }
  SizeBox_0 {
   widgetClassPath=&SizeBox
   parent=Overlay_35
   widget=SizeBox_0
   properties {bOverride_HeightOverride=true, bOverride_WidthOverride=true, heightOverride=128, widthOverride=128}
   slot {class=OverlaySlot, properties {horizontalAlignment=HAlign_Center, verticalAlignment=VAlign_Center}}
  }
  Overlay_37 {widgetClassPath=&Overlay, parent=SizeBox_0, widget=Overlay_37, properties {visibility=Visible}, slot {class=SizeBoxSlot}}
  Image_Cooldownbg {
   widgetClassPath=&Image
   bIsVariable=true
   parent=Overlay_37
   widget=Image_Cooldownbg
   properties {
    renderOpacity=0
    visibility=HitTestInvisible
    brush {
     bIsDynamicallyLoaded=false
     drawAs=Image
     imageType=NoImage
     mirroring=NoMirror
     resourceName=None
     resourceObject=/HighTower/UI/Materials/M_UI_AbilityCooldownEffect_Inst
     tiling=NoTile
     imageSize=<500 343.6225>
     margin=<ltrb 0 0 0 0>
     outlineSettings {
      bUseBrushTransparency=false
      roundingType=HalfHeightRadius
      width=0
      color {colorUseRule=UseColor_Specified, specifiedColor=<rgba 0 0 0 0>}
      cornerRadii=<0 0 0 0>
     }
     tintColor {colorUseRule=UseColor_Specified, specifiedColor=<rgba 1 1 1 1>}
     uVRegion {bIsValid=false, max=<0 0>, min=<0 0>}
    }
    renderTransform {angle=-90, scale=<1 1>, shear=<0 0>, translation=<0 -161>}
   }
   slot {class=OverlaySlot, properties {horizontalAlignment=HAlign_Center, verticalAlignment=VAlign_Center, padding=<ltrb -250 -250 -250 -250>}}
  }
  Image_InteractPulse {
   widgetClassPath=&Image
   bIsVariable=true
   parent=Overlay_37
   widget=Image_InteractPulse
   properties {
    visibility=Collapsed
    brush {
     bIsDynamicallyLoaded=false
     drawAs=Image
     imageType=NoImage
     mirroring=NoMirror
     resourceName=None
     resourceObject=&M_UI_GameplayDiamondPing_Pulse_Gradient_Inst
     tiling=NoTile
     imageSize=<200 200>
     margin=<ltrb 0 0 0 0>
     outlineSettings {
      bUseBrushTransparency=false
      roundingType=HalfHeightRadius
      width=0
      color {colorUseRule=UseColor_Specified, specifiedColor=<rgba 0 0 0 0>}
      cornerRadii=<0 0 0 0>
     }
     tintColor {colorUseRule=UseColor_Specified, specifiedColor=<rgba 1 1 1 1>}
     uVRegion {bIsValid=false, max=<0 0>, min=<0 0>}
    }
   }
   slot {class=OverlaySlot, properties {horizontalAlignment=HAlign_Center, verticalAlignment=VAlign_Center, padding=<ltrb -128 -128 -128 -128>}}
  }
  Image_AbilityIcon {
   widgetClassPath=&Image
   bIsVariable=true
   parent=Overlay_37
   widget=Image_AbilityIcon
   properties {
    visibility=SelfHitTestInvisible
    brush {
     bIsDynamicallyLoaded=false
     drawAs=Image
     imageType=NoImage
     mirroring=NoMirror
     resourceName=None
     resourceObject=/HighTower/UI/Materials/M_UI_AbilitySocket_Inst
     tiling=NoTile
     imageSize=<145 145>
     margin=<ltrb 0 0 0 0>
     outlineSettings {
      bUseBrushTransparency=false
      roundingType=HalfHeightRadius
      width=0
      color {colorUseRule=UseColor_Specified, specifiedColor=<rgba 0 0 0 0>}
      cornerRadii=<0 0 0 0>
     }
     tintColor {colorUseRule=UseColor_Specified, specifiedColor=<rgba 1 1 1 1>}
     uVRegion {bIsValid=false, max=<0 0>, min=<0 0>}
    }
   }
   slot {class=OverlaySlot, properties {horizontalAlignment=HAlign_Center, verticalAlignment=VAlign_Center, padding=<ltrb -128 -128 -128 -128>}}
  }
  QuickbarSlotCooldown {
   widgetClassPath=/Game/UI/InGame/HUD/QuickbarSlotCooldown
   bIsVariable=true
   parent=Overlay_37
   widget=QuickbarSlotCooldown
   properties {renderOpacity=0, visibility=Collapsed}
   slot {class=OverlaySlot}
  }
  Image_Charge {
   widgetClassPath=&Image
   bIsVariable=true
   parent=Overlay_37
   widget=Image_Charge
   properties {
    visibility=HitTestInvisible
    brush {
     bIsDynamicallyLoaded=false
     drawAs=Image
     imageType=NoImage
     mirroring=NoMirror
     resourceName=None
     resourceObject=/HighTower/UI/Materials/M_UI_AbilityCharge_Inst
     tiling=NoTile
     imageSize=<145 145>
     margin=<ltrb 0 0 0 0>
     outlineSettings {
      bUseBrushTransparency=false
      roundingType=HalfHeightRadius
      width=0
      color {colorUseRule=UseColor_Specified, specifiedColor=<rgba 0 0 0 0>}
      cornerRadii=<0 0 0 0>
     }
     tintColor {colorUseRule=UseColor_Specified, specifiedColor=<rgba 1 1 1 1>}
     uVRegion {bIsValid=false, max=<0 0>, min=<0 0>}
    }
   }
   slot {class=OverlaySlot, properties {horizontalAlignment=HAlign_Center, verticalAlignment=VAlign_Center, padding=<ltrb -200 -200 -200 -200>}}
  }
  CommonBorder_Input {
   widgetClassPath=&CommonBorder
   parent=Overlay_37
   widget=CommonBorder_Input
   properties {padding=<ltrb 0 0 0 0>}
   slot {class=OverlaySlot, properties {horizontalAlignment=HAlign_Fill, verticalAlignment=VAlign_Fill}}
  }
  Overlay_0 {
   widgetClassPath=&Overlay
   parent=CommonBorder_Input
   widget=Overlay_0
   slot {class=BorderSlot, properties {padding=<ltrb 0 0 0 0>}}
  }
  KeybindWidget {
   widgetClassPath=/FortUILibrary/Input/KeybindWidget
   bIsVariable=true
   parent=Overlay_0
   widget=KeybindWidget
   slot {class=OverlaySlot, properties {verticalAlignment=VAlign_Bottom}}
  }
  KeybindWidget_Alt {
   widgetClassPath=&KeybindWidget
   bIsVariable=true
   parent=Overlay_0
   widget=KeybindWidget_Alt
   slot {class=OverlaySlot, properties {horizontalAlignment=HAlign_Right, verticalAlignment=VAlign_Bottom}}
  }
  Text_Cooldown {
   widgetClassPath=&CommonTextBlock
   bIsVariable=true
   parent=Overlay_37
   widget=Text_Cooldown
   properties {
    renderOpacity=0
    text=5
    font {
     bForceMonospaced=false
     bMaterialIsStencil=false
     fontMaterial=null
     fontObject=&Burbank
     letterSpacing=0
     monospacedWidth=1
     size=30
     skewAmount=0
     typefaceFontName=BurbankRegular-Black
     outlineSettings {
      bApplyOutlineToDropShadows=false
      bMiteredCorners=false
      bSeparateFillAlpha=false
      outlineMaterial=null
      outlineSize=3
      outlineColor=<rgba 0 0 0 1>
     }
    }
    renderTransform {angle=0, scale=<1 1>, shear=<-8 0>, translation=<3 0>}
    shadowColorAndOpacity=<rgba 0 0 0 0.72>
    shadowOffset=<8 8>
   }
   slot {class=OverlaySlot, properties {horizontalAlignment=HAlign_Center, verticalAlignment=VAlign_Center}}
  }
  Image_Spark1 {
   widgetClassPath=&Image
   bIsVariable=true
   parent=Overlay_37
   widget=Image_Spark1
   properties {
    renderOpacity=0
    visibility=SelfHitTestInvisible
    brush {
     bIsDynamicallyLoaded=false
     drawAs=Image
     imageType=NoImage
     mirroring=NoMirror
     resourceName=None
     resourceObject=/Game/Athena/HUD/Materials/M_Electric_Lightning_Arcs_Inst
     tiling=NoTile
     imageSize=<256 256>
     margin=<ltrb 0 0 0 0>
     outlineSettings {
      bUseBrushTransparency=false
      roundingType=HalfHeightRadius
      width=0
      color {colorUseRule=UseColor_Specified, specifiedColor=<rgba 0 0 0 0>}
      cornerRadii=<0 0 0 0>
     }
     tintColor {colorUseRule=UseColor_Specified, specifiedColor=<rgba 1 1 1 1>}
     uVRegion {bIsValid=false, max=<0 0>, min=<0 0>}
    }
   }
   slot {class=OverlaySlot, properties {horizontalAlignment=HAlign_Center, verticalAlignment=VAlign_Center, padding=<ltrb -256 -256 -256 -256>}}
  }
  Image_Spark1_1 {
   widgetClassPath=&Image
   bIsVariable=true
   parent=Overlay_37
   widget=Image_Spark1_1
   properties {
    renderOpacity=0
    visibility=SelfHitTestInvisible
    brush {
     bIsDynamicallyLoaded=false
     drawAs=Image
     imageType=NoImage
     mirroring=NoMirror
     resourceName=None
     resourceObject=/Game/Athena/HUD/Materials/M_Electric_Lightning_Arcs_Inst_Inst
     tiling=NoTile
     imageSize=<64 64>
     margin=<ltrb 0 0 0 0>
     outlineSettings {
      bUseBrushTransparency=false
      roundingType=HalfHeightRadius
      width=0
      color {colorUseRule=UseColor_Specified, specifiedColor=<rgba 0 0 0 0>}
      cornerRadii=<0 0 0 0>
     }
     tintColor {colorUseRule=UseColor_Specified, specifiedColor=<rgba 1 1 1 1>}
     uVRegion {bIsValid=false, max=<0 0>, min=<0 0>}
    }
    renderTransform {angle=123, scale=<1 1>, shear=<0 0>, translation=<54 0>}
   }
   slot {class=OverlaySlot, properties {horizontalAlignment=HAlign_Center, verticalAlignment=VAlign_Center, padding=<ltrb -256 -256 -256 -256>}}
  }
  Image_Spark1_2 {
   widgetClassPath=&Image
   bIsVariable=true
   parent=Overlay_37
   widget=Image_Spark1_2
   properties {
    renderOpacity=0
    visibility=SelfHitTestInvisible
    brush {
     bIsDynamicallyLoaded=false
     drawAs=Image
     imageType=NoImage
     mirroring=NoMirror
     resourceName=None
     resourceObject=/Game/Athena/HUD/Materials/M_Electric_Lightning_Arcs_Inst_Inst1
     tiling=NoTile
     imageSize=<128 200>
     margin=<ltrb 0 0 0 0>
     outlineSettings {
      bUseBrushTransparency=false
      roundingType=HalfHeightRadius
      width=0
      color {colorUseRule=UseColor_Specified, specifiedColor=<rgba 0 0 0 0>}
      cornerRadii=<0 0 0 0>
     }
     tintColor {colorUseRule=UseColor_Specified, specifiedColor=<rgba 1 1 1 1>}
     uVRegion {bIsValid=false, max=<0 0>, min=<0 0>}
    }
    renderTransform {angle=0, scale=<1 1>, shear=<0 0>, translation=<2 -146>}
   }
   slot {class=OverlaySlot, properties {horizontalAlignment=HAlign_Center, verticalAlignment=VAlign_Center, padding=<ltrb -256 -256 -256 -256>}}
  }
  Image_Spark1_3 {
   widgetClassPath=&Image
   bIsVariable=true
   parent=Overlay_37
   widget=Image_Spark1_3
   properties {
    renderOpacity=0
    visibility=SelfHitTestInvisible
    brush {
     bIsDynamicallyLoaded=false
     drawAs=Image
     imageType=NoImage
     mirroring=NoMirror
     resourceName=None
     resourceObject=/Game/Athena/HUD/Materials/M_Electric_Lightning_Arcs_Inst_Inst1_Inst
     tiling=NoTile
     imageSize=<180 180>
     margin=<ltrb 0 0 0 0>
     outlineSettings {
      bUseBrushTransparency=false
      roundingType=HalfHeightRadius
      width=0
      color {colorUseRule=UseColor_Specified, specifiedColor=<rgba 0 0 0 0>}
      cornerRadii=<0 0 0 0>
     }
     tintColor {colorUseRule=UseColor_Specified, specifiedColor=<rgba 1 1 1 1>}
     uVRegion {bIsValid=false, max=<0 0>, min=<0 0>}
    }
    renderTransform {angle=-105, scale=<1 1>, shear=<0 0>, translation=<0 0>}
   }
   slot {class=OverlaySlot, properties {horizontalAlignment=HAlign_Center, verticalAlignment=VAlign_Center, padding=<ltrb -256 -256 -256 -256>}}
  }
  Image_Spark1_4 {
   widgetClassPath=&Image
   bIsVariable=true
   parent=Overlay_37
   widget=Image_Spark1_4
   properties {
    renderOpacity=0
    visibility=SelfHitTestInvisible
    brush {
     bIsDynamicallyLoaded=false
     drawAs=Image
     imageType=NoImage
     mirroring=NoMirror
     resourceName=None
     resourceObject=&M_Electric_Lightning_Arcs_Inst_Inst1_Inst
     tiling=NoTile
     imageSize=<128 128>
     margin=<ltrb 0 0 0 0>
     outlineSettings {
      bUseBrushTransparency=false
      roundingType=HalfHeightRadius
      width=0
      color {colorUseRule=UseColor_Specified, specifiedColor=<rgba 0 0 0 0>}
      cornerRadii=<0 0 0 0>
     }
     tintColor {colorUseRule=UseColor_Specified, specifiedColor=<rgba 1 1 1 1>}
     uVRegion {bIsValid=false, max=<0 0>, min=<0 0>}
    }
    renderTransform {angle=0, scale=<1 1>, shear=<0 0>, translation=<2 -306>}
   }
   slot {class=OverlaySlot, properties {horizontalAlignment=HAlign_Center, verticalAlignment=VAlign_Center, padding=<ltrb -256 -256 -256 -256>}}
  }
  Image_Glow {
   widgetClassPath=&Image
   bIsVariable=true
   parent=Overlay_37
   widget=Image_Glow
   properties {
    renderOpacity=0
    visibility=SelfHitTestInvisible
    brush {
     bIsDynamicallyLoaded=false
     drawAs=Image
     imageType=NoImage
     mirroring=NoMirror
     resourceName=None
     resourceObject=/Game/Athena/HUD/LTM/Wax/Materials/M_UI_TextGlow_Add_Gradient_Gold
     tiling=NoTile
     imageSize=<256 256>
     margin=<ltrb 0 0 0 0>
     outlineSettings {
      bUseBrushTransparency=false
      roundingType=HalfHeightRadius
      width=0
      color {colorUseRule=UseColor_Specified, specifiedColor=<rgba 0 0 0 0>}
      cornerRadii=<0 0 0 0>
     }
     tintColor {colorUseRule=UseColor_Specified, specifiedColor=<rgba 1 0.3984 0 1>}
     uVRegion {bIsValid=false, max=<0 0>, min=<0 0>}
    }
   }
   slot {class=OverlaySlot, properties {horizontalAlignment=HAlign_Center, verticalAlignment=VAlign_Center, padding=<ltrb -128 -128 -128 -128>}}
  }
 ]
}
