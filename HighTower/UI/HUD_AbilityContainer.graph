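# HighTower HUD element that lays out the Primary, Secondary, Utility and Jump ability widgets, each bound to a power-type tag, at the bottom center of the screen.
# On Construct it raises the bottom padding for mobile games, then collapses itself during replays.
# Otherwise it shows the ability widgets, keeping PrimaryAlt, which shares the Primary tag, collapsed.
# A platform visibility wrapper hides the mobile switch-mode button on PC, console and gamepad.
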WidgetBlueprint /HighTower/UI/HUD_AbilityContainer
graphs [EventGraph {nodes [
   Event Construct()
   > n1 = Branch(Condition=IsMobileGame())
   > n2 = SetPadding!(
    execute
    self:Overlay Slot Object Reference=SlotAsOverlaySlot(Widget=Get ScaleBox_1() -> (ScaleBox_1:Scale Box Object Reference))
    InPadding:Margin Structure=Make Margin(Left=0, Top=0, Right=0, Bottom=150)
   ) -> (then)
   n3 = Branch(execute=[@n2.then, @n1.else], Condition=GetIsPlayingReplay!() -> (ReturnValue:bool))
   > SetVisibility(InVisibility=Collapsed)
   SetVisibility(execute=@n3.else, InVisibility=SelfHitTestInvisible)
   > SetVisibility(
    self=[
     Get HUD_AbilityWidget_Jump() -> (HUD_AbilityWidget_Jump:HUD Ability Widget Object Reference)
     Get HUD_AbilityWidget_Primary() -> (HUD_AbilityWidget_Primary:HUD Ability Widget Object Reference)
     Get HUD_AbilityWidget_Secondary() -> (HUD_AbilityWidget_Secondary:HUD Ability Widget Object Reference)
     Get HUD_AbilityWidget_Utility() -> (HUD_AbilityWidget_Utility:HUD Ability Widget Object Reference)
     Get ScaleBox_1()
    ]
    InVisibility=SelfHitTestInvisible
   )
   > SetVisibility(self=Get HUD_AbilityWidget_PrimaryAlt() -> (HUD_AbilityWidget_PrimaryAlt:HUD Ability Widget Object Reference), InVisibility=Collapsed)
  ]}]
variables []
widgets {
 info {inheritedWidgetCount=1, parentClass=&FortHUDElementWidget, rootWidgetClass=&Overlay, widgetCount=15}
 widgets [
  Overlay_0 {widgetClassPath=&Overlay, widget=Overlay_0}
  ScaleBox_1 {
   widgetClassPath=&ScaleBox
   bIsVariable=true
   parent=Overlay_0
   widget=ScaleBox_1
   properties {stretch=UserSpecified, userSpecifiedScale=0.8}
   slot {class=OverlaySlot, properties {horizontalAlignment=HAlign_Center, verticalAlignment=VAlign_Bottom, padding=<ltrb 0 0 0 30>}}
  }
  VerticalBox_16 {widgetClassPath=&VerticalBox, parent=ScaleBox_1, widget=VerticalBox_16, slot {class=ScaleBoxSlot}}
  HorizontalBox_32 {
   widgetClassPath=&HorizontalBox
   parent=VerticalBox_16
   widget=HorizontalBox_32
   slot {class=VerticalBoxSlot, properties {horizontalAlignment=HAlign_Center, verticalAlignment=VAlign_Center}}
  }
  HUD_AbilityWidget_PrimaryAlt {
   widgetClassPath=&/HUD_AbilityWidget
   bIsVariable=true
   parent=HorizontalBox_32
   widget=HUD_AbilityWidget_PrimaryAlt
   properties {primaryAltAbility=true, visibility=Collapsed, actionTag {tagName=Trait.HighTower.Power.Type.Primary}, padding=<ltrb 20 0 20 0>}
   slot {class=HorizontalBoxSlot, properties {horizontalAlignment=HAlign_Right, verticalAlignment=VAlign_Top}}
  }
  HUD_AbilityWidget_Primary {
   widgetClassPath=&HUD_AbilityWidget
   bIsVariable=true
   parent=HorizontalBox_32
   widget=HUD_AbilityWidget_Primary
   properties {primaryAbility=true, actionTag {tagName=Trait.HighTower.Power.Type.Primary}, padding=<ltrb 20 0 20 0>}
   slot {class=HorizontalBoxSlot, properties {horizontalAlignment=HAlign_Right, verticalAlignment=VAlign_Top}}
  }
  ScaleBox_0 {
   widgetClassPath=&ScaleBox
   parent=VerticalBox_16
   widget=ScaleBox_0
   properties {stretch=UserSpecified, userSpecifiedScale=0.85}
   slot {class=VerticalBoxSlot, properties {horizontalAlignment=HAlign_Center, verticalAlignment=VAlign_Center}}
  }
  Overlay_1 {widgetClassPath=&Overlay, parent=ScaleBox_0, widget=Overlay_1, slot {class=ScaleBoxSlot}}
  HorizontalBox {widgetClassPath=&HorizontalBox, parent=Overlay_1, widget=HorizontalBox, slot {class=OverlaySlot}}
  HUD_AbilityWidget_Utility {
   widgetClassPath=&HUD_AbilityWidget
   bIsVariable=true
   parent=HorizontalBox
   widget=HUD_AbilityWidget_Utility
   properties {actionTag {tagName=Trait.HighTower.Power.Type.Utility}}
   slot {class=HorizontalBoxSlot, properties {horizontalAlignment=HAlign_Left, verticalAlignment=VAlign_Top, padding=<ltrb 0 0 0 20>}}
  }
  HUD_AbilityWidget_Jump {
   widgetClassPath=&HUD_AbilityWidget
   bIsVariable=true
   parent=HorizontalBox
   widget=HUD_AbilityWidget_Jump
   properties {actionTag {tagName=Trait.HighTower.Power.Type.Jump}}
   slot {class=HorizontalBoxSlot, properties {horizontalAlignment=HAlign_Left, verticalAlignment=VAlign_Bottom, padding=<ltrb 40 0 40 0>}}
  }
  HUD_AbilityWidget_Secondary {
   widgetClassPath=&HUD_AbilityWidget
   bIsVariable=true
   parent=HorizontalBox
   widget=HUD_AbilityWidget_Secondary
   properties {actionTag {tagName=Trait.HighTower.Power.Type.Secondary}}
   slot {class=HorizontalBoxSlot, properties {horizontalAlignment=HAlign_Left, verticalAlignment=VAlign_Top, padding=<ltrb 0 0 0 20>}}
  }
  CommonVisibilityWidget_0 {
   widgetClassPath=&CommonVisibilityWidgetLegacy
   parent=Overlay_1
   widget=CommonVisibilityWidget_0
   properties {
    bShowForErebus=false
    bShowForGamepad=false
    bShowForMac=false
    bShowForMouseAndKeyboard=false
    bShowForPC=false
    bShowForPS4=false
    bShowForPS5=false
    bShowForSwitch2=false
    bShowForXBox=false
    bShowForXSX=false
   }
   slot {class=OverlaySlot, properties {horizontalAlignment=HAlign_Center, verticalAlignment=VAlign_Center, padding=<ltrb 325 0 0 0>}}
  }
  MobileSwitchModeButton {
   widgetClassPath=&/MobileSwitchModeButton
   bIsVariable=true
   parent=CommonVisibilityWidget_0
   widget=MobileSwitchModeButton
   properties {visibility=SelfHitTestInvisible}
   slot {class=BorderSlot}
  }
  TutorialHighlight {widgetClassPath=&TutorialHighlight}
 ]
}
