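# HighTower HUD interaction tooltip that shows an ability pickup's name, short description and secondary info.
# PreConstruct uses GameplayTag to pick the background material and rich-text style, applies a lazy texture, and fills secondary text from AttachedAbility (hard-coded "Double Jump" suffix) or AdditionalInfo.
# Context updates cast the receiving actor to FortPickup, set the title from its item and show the widget.
# GetInteractText returns pickup text with the success flag inverted.
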WidgetBlueprint /HighTower/UI/HUD_AbilityInteractionContext
graphs [
 EventGraph {nodes [
   Event PreConstruct()
   n1 = Get GameplayTag() -> (GameplayTag:Gameplay Tag Structure)
   n2 = Find(
    TargetMap:Map of Gameplay Tag Structures to Material Instance Object References=Get GameplayTagMap() -> (GameplayTagMap:Map of Gameplay Tag Structures to Material Instance Object References)
    Key:"Gameplay Tag Structure (by ref)"=@n1
   ) -> (Value:Material Instance Object Reference)
   > Branch(Condition=AND Boolean(A=IsGameplayTagValid!(GameplayTag:Gameplay Tag Structure=@n1) -> (ReturnValue:bool), B=@n2.ReturnValue))
   n3 = Get Image_TooltipBG() -> (Image_TooltipBG:Fort Lazy Image Object Reference)
   > SetBrushFromMaterial(self=@n3, Material=@n2)
   > n4 = GetDynamicMaterial(self=@n3)
   > Set BG_MID(BG_MID:Material Instance Dynamic Object Reference=@n4) -> (Output_Get:Material Instance Dynamic Object Reference)
   n5 = Get GameplayTag()
   n6 = Get GameplayTagRichText() -> (GameplayTagRichText:Map of Gameplay Tag Structures to Strings)
   > SetText(
    self=Get RichText_Subtitle() -> (RichText_Subtitle:Common Rich Text Block Object Reference)
    InText=Format Text!(
     Format="LOCTEXT(\"<{RT}>{Description}</>\")"
     RT:String=Find(TargetMap:Map of Gameplay Tag Structures to Strings=@n6, Key=@n5) -> (Value:String)
     Description:Text=Get ShortDescription() -> (ShortDescription:Text)
    ) -> (Result)
   )
   n7 = Get Image_TooltipBG()
   > SetMaterialTextureParamName!(execute, self:Common Lazy Image Object Reference=@n7, TextureParamName:Name=Texture) -> (then)
   > SetBrushFromLazyTexture!(
    execute
    self:Common Lazy Image Object Reference=@n7
    LazyTexture:"Texture 2D Soft Object Reference (by ref)"=Get TooltipTexture() -> (TooltipTexture:Texture 2D Soft Object Reference)
    bMatchSize:bool=false
   ) -> (then)
   > n8 = Branch(Condition=IsValid(Object=Get AttachedAbility() -> (AttachedAbility:Fort Item Definition Object Reference)))
   n9 = Get AttachedAbility()
   n10 = GetDisplayName!(self:Item Definition Base Object Reference=@n9) -> (ReturnValue:Text)
   n11 = Find(TargetMap=Get GameplayTagRichText(), Key=Get GameplayTag())
   n12 = Get AdditionalInfo() -> (AdditionalInfo:Text)
   n13 = Get RichText_Secondary() -> (RichText_Secondary:Common Rich Text Block Object Reference)
   > SetText(
    self=@n13
    InText=Select(
     Option 0:"Text (by ref)"=@n12
     Option 1:"Text (by ref)"=Format Text!(Format="LOCTEXT(\"+{Name} <{RT}>(Double Jump)</>\")", Name:Text=@n10, RT=@n11) -> (Result)
     Index=IsValid(Object=@n9)
    ) -> (ReturnValue:"Text (by ref)")
   )
   SetText(execute=@n8.else, self=@n13, InText=@n12)
   n14 = Event OnContextInfoUpdated!() -> (OutputDelegate, then, ContextInfo:Fort Interact Context Info Object Reference) {is=title K2Node_Event}
   > n15 = Set CurrentInteraction(CurrentInteraction:Fort Interact Context Info Object Reference=@n14.ContextInfo) -> (Output_Get:Fort Interact Context Info Object Reference)
   > n16 = Cast To FortPickup!(execute, Object=Get ReceivingActor!(self:Fort Interact Context Info Object Reference=@n15) -> (ReceivingActor:Actor Weak object)) -> (then, CastFailed, AsFort Pickup:Fort Pickup Object Reference) {is=title K2Node_DynamicCast}
   n17 = Get PrimaryPickupDummyItem!(self:Fort Pickup Object Reference=@n16) -> (PrimaryPickupDummyItem:Fort Item Object Reference)
   n18 = GetItemDefinition!(self:Fort Item Object Reference=@n17) -> (ReturnValue:Fort Item Definition Object Reference)
   > SetText(
    self=Get RichText_Title() -> (RichText_Title:Common Rich Text Block Object Reference)
    InText=Format Text!(Format="LOCTEXT(\"<title>{0}</>\")", 0:Text=GetDisplayName(self=@n18)) -> (Result)
   )
   > SetVisibility(InVisibility=HitTestInvisible)
   n19 = GetDebugStringFromGameplayTag!(GameplayTag:Gameplay Tag Structure) -> (ReturnValue:String)
   Find(
    TargetMap:Map of Strings to Material Instance Object References=Get StringTagMap() -> (StringTagMap:Map of Strings to Material Instance Object References)
    Key:"String (by ref)"=Append(A=Power, B=Split(SourceString=@n19, InStr=Power, SearchDir=FromStart).RightS)
   ) -> (Value:Material Instance Object Reference)
   GetActionTagForItem!(
    self:Fort Inventory Subsystem Object Reference=Get FortInventorySubsystem() -> (ReturnValue:Fort Inventory Subsystem Object Reference) {is=title K2Node_GetSubsystem}
    Item:Fort Item Object Reference=@n17
   ) -> (ReturnValue:Gameplay Tag Structure)
  ]}
 GetInteractText {nodes [
   GetInteractText() {is=title K2Node_FunctionEntry}
   > n1 = Cast To FortPickup(Object=Get ReceivingActor(self=Get CurrentInteraction() -> (CurrentInteraction:Fort Interact Context Info Object Reference)))
   > n2 = GetInteractText!(execute, self:Fort Pickup Object Reference=@n1) -> (then, InteractText:Text, ReturnValue:bool) {is=title K2Node_CallFunction}
   > Return Node!(execute, InteractText:Text=@n2, Result:bool=NOT Boolean(A=@n2.ReturnValue))
   Return Node(execute=@n1.CastFailed, Result=false)
  ]}
]
variables [
 AdditionalInfo {type=string, default="+Additional ability info <bs>(Context)</>"}
 AttachedAbility {type=FortItemDefinition, default=null}
 BG_MID {type=MaterialInstanceDynamic}
 CurrentInteraction {type=FortInteractContextInfo, default=null}
 GameplayTag {type=GameplayTag, default {tagName=None}}
 GameplayTagMap {default {
   "(TagName=\"Trait.HighTower.Power.Type.Jump\")"=null
   "(TagName=\"Trait.HighTower.Power.Type.Primary\")"=null
   "(TagName=\"Trait.HighTower.Power.Type.Secondary\")"=null
   "(TagName=\"Trait.HighTower.Power.Type.Utility\")"=null
  }}
 GameplayTagRichText {default {
   "(TagName=\"Trait.HighTower.Power.Type.Jump\")"=bj
   "(TagName=\"Trait.HighTower.Power.Type.Primary\")"=bp
   "(TagName=\"Trait.HighTower.Power.Type.Secondary\")"=bs
   "(TagName=\"Trait.HighTower.Power.Type.Utility\")"=bu
  }}
 ShortDescription {type=string, default="<bs>(Super brief tooltip)</>"}
 StringTagMap {type="map<MaterialInstance>", default {Power.Type.Jump=null, Power.Type.Primary=null, Power.Type.Secondary=null, Power.Type.Utility=null}}
 TooltipTexture {type=Texture2D, default=/HighTower/UI/Materials/T_UI_Hightower_TooltipBG_Squash}
]
widgets {
 info {inheritedWidgetCount=0, parentClass=/Script/FortniteGame.FortInteractContextInfoWidget, rootWidgetClass=&Overlay, widgetCount=13}
 widgets [
  Overlay_1 {widgetClassPath=&Overlay, widget=Overlay_1}
  Overlay_22 {widgetClassPath=&Overlay, parent=Overlay_1, widget=Overlay_22, slot {class=OverlaySlot}}
  Image_shadow {
   widgetClassPath=&Image
   bIsVariable=true
   parent=Overlay_22
   widget=Image_shadow
   properties {
    brush {
     bIsDynamicallyLoaded=false
     drawAs=Image
     imageType=NoImage
     mirroring=NoMirror
     resourceName=None
     resourceObject=null
     tiling=NoTile
     imageSize=<32 32>
     margin=<ltrb 0 0 0 0>
     outlineSettings {
      bUseBrushTransparency=false
      roundingType=HalfHeightRadius
      width=0
      color {colorUseRule=UseColor_Specified, specifiedColor=<rgba 0 0 0 0>}
      cornerRadii=<0 0 0 0>
     }
     tintColor {colorUseRule=UseColor_Specified, specifiedColor=<rgba 0 0 0 0.4>}
     uVRegion {bIsValid=false, max=<0 0>, min=<0 0>}
    }
    renderTransform {angle=0, scale=<1 1>, shear=<0 0>, translation=<10 10>}
   }
   slot {class=OverlaySlot, properties {horizontalAlignment=HAlign_Fill, verticalAlignment=VAlign_Fill}}
  }
  VerticalBox_0 {widgetClassPath=&VerticalBox, parent=Overlay_22, widget=VerticalBox_0, slot {class=OverlaySlot}}
  Overlay_0 {widgetClassPath=&Overlay, parent=VerticalBox_0, widget=Overlay_0, slot {class=VerticalBoxSlot}}
  Image_BG {
   widgetClassPath=&Image
   bIsVariable=true
   parent=Overlay_0
   widget=Image_BG
   properties {brush {
     bIsDynamicallyLoaded=false
     drawAs=Image
     imageType=NoImage
     mirroring=NoMirror
     resourceName=None
     resourceObject=null
     tiling=NoTile
     imageSize=<32 32>
     margin=<ltrb 0 0 0 0>
     outlineSettings {
      bUseBrushTransparency=false
      roundingType=HalfHeightRadius
      width=0
      color {colorUseRule=UseColor_Specified, specifiedColor=<rgba 0 0 0 0>}
      cornerRadii=<0 0 0 0>
     }
     tintColor {colorUseRule=UseColor_Specified, specifiedColor=<rgba 0.0273 0.0273 0.0273 1>}
     uVRegion {bIsValid=false, max=<0 0>, min=<0 0>}
    }}
   slot {class=OverlaySlot, properties {horizontalAlignment=HAlign_Fill, verticalAlignment=VAlign_Fill}}
  }
  Image_TooltipBG {
   widgetClassPath=&FortLazyImage
   bIsVariable=true
   parent=Overlay_0
   widget=Image_TooltipBG
   properties {materialTextureParamName=Texture}
   slot {class=OverlaySlot, properties {horizontalAlignment=HAlign_Fill, verticalAlignment=VAlign_Fill}}
  }
  SizeBox_0 {
   widgetClassPath=&SizeBox
   parent=Overlay_0
   widget=SizeBox_0
   properties {bOverride_MinDesiredHeight=true, bOverride_MinDesiredWidth=true, minDesiredHeight=112, minDesiredWidth=390}
   slot {class=OverlaySlot}
  }
  CommonBorder_0 {
   widgetClassPath=&CommonBorder
   parent=SizeBox_0
   widget=CommonBorder_0
   properties {padding=<ltrb 10 6 6 6>}
   slot {class=SizeBoxSlot}
  }
  VerticalBox_1 {
   widgetClassPath=&VerticalBox
   parent=CommonBorder_0
   widget=VerticalBox_1
   slot {class=BorderSlot, properties {padding=<ltrb 10 6 6 6>}}
  }
  RichText_Title {
   widgetClassPath=&CommonRichTextBlock
   bIsVariable=true
   parent=VerticalBox_1
   widget=RichText_Title
   properties {
    text="<title>Ability Name</>"
    textStyleSet=/Game/Athena/HUD/Abilities/AbilityRichText
    textTransformPolicy=ToUpper
    defaultTextStyleOverride {
     overflowPolicy=Clip
     transformPolicy=None
     colorAndOpacity {colorUseRule=UseColor_Specified, specifiedColor=<rgba 1 1 1 1>}
     font {
      bForceMonospaced=false
      bMaterialIsStencil=false
      fontMaterial=null
      fontObject=&Burbank
      letterSpacing=0
      monospacedWidth=1
      size=20
      skewAmount=0
      typefaceFontName=BurbankRegular-Black
      outlineSettings {
       bApplyOutlineToDropShadows=false
       bMiteredCorners=false
       bSeparateFillAlpha=false
       outlineMaterial=null
       outlineSize=0
       outlineColor=<rgba 0 0 0 1>
      }
     }
     highlightColor {colorUseRule=UseColor_Specified, specifiedColor=<rgba 0 0 0 1>}
     highlightShape {
      bIsDynamicallyLoaded=false
      drawAs=Image
      imageType=NoImage
      mirroring=NoMirror
      resourceName=None
      resourceObject=null
      tiling=NoTile
      imageSize=<32 32>
      margin=<ltrb 0 0 0 0>
      outlineSettings {
       bUseBrushTransparency=false
       roundingType=HalfHeightRadius
       width=0
       color {colorUseRule=UseColor_Specified, specifiedColor=<rgba 0 0 0 0>}
       cornerRadii=<0 0 0 0>
      }
      tintColor {colorUseRule=UseColor_Specified, specifiedColor=<rgba 1 1 1 1>}
      uVRegion {bIsValid=false, max=<0 0>, min=<0 0>}
     }
     selectedBackgroundColor {colorUseRule=UseColor_Foreground, specifiedColor=<rgba 1 0 1 1>}
     shadowColorAndOpacity=<rgba 0 0 0 1>
     shadowOffset=<0 0>
     strikeBrush {
      bIsDynamicallyLoaded=false
      drawAs=Image
      imageType=NoImage
      mirroring=NoMirror
      resourceName=None
      resourceObject=null
      tiling=NoTile
      imageSize=<32 32>
      margin=<ltrb 0 0 0 0>
      outlineSettings {
       bUseBrushTransparency=false
       roundingType=HalfHeightRadius
       width=0
       color {colorUseRule=UseColor_Specified, specifiedColor=<rgba 0 0 0 0>}
       cornerRadii=<0 0 0 0>
      }
      tintColor {colorUseRule=UseColor_Specified, specifiedColor=<rgba 1 1 1 1>}
      uVRegion {bIsValid=false, max=<0 0>, min=<0 0>}
     }
     underlineBrush {
      bIsDynamicallyLoaded=false
      drawAs=Image
      imageType=NoImage
      mirroring=NoMirror
      resourceName=None
      resourceObject=null
      tiling=NoTile
      imageSize=<32 32>
      margin=<ltrb 0 0 0 0>
      outlineSettings {
       bUseBrushTransparency=false
       roundingType=HalfHeightRadius
       width=0
       color {colorUseRule=UseColor_Specified, specifiedColor=<rgba 0 0 0 0>}
       cornerRadii=<0 0 0 0>
      }
      tintColor {colorUseRule=UseColor_Specified, specifiedColor=<rgba 1 1 1 1>}
      uVRegion {bIsValid=false, max=<0 0>, min=<0 0>}
     }
    }
    renderTransform {angle=0, scale=<1 1>, shear=<-8 0>, translation=<0 0>}
   }
   slot {class=VerticalBoxSlot, properties {padding=<ltrb 0 0 100 -8>}}
  }
  RichText_Subtitle {
   widgetClassPath=&CommonRichTextBlock
   bIsVariable=true
   parent=VerticalBox_1
   widget=RichText_Subtitle
   properties {
    autoWrapText=true
    lineHeightPercentage=0.85
    text="<bs>(Super brief tooltip)</>"
    textStyleSet=&AbilityRichText
    textTransformPolicy=ToUpper
    defaultTextStyleOverride {
     overflowPolicy=Clip
     transformPolicy=None
     colorAndOpacity {colorUseRule=UseColor_Specified, specifiedColor=<rgba 1 1 1 1>}
     font {
      bForceMonospaced=false
      bMaterialIsStencil=false
      fontMaterial=null
      fontObject=&Burbank
      letterSpacing=0
      monospacedWidth=1
      size=20
      skewAmount=0
      typefaceFontName=BurbankRegular-Black
      outlineSettings {
       bApplyOutlineToDropShadows=false
       bMiteredCorners=false
       bSeparateFillAlpha=false
       outlineMaterial=null
       outlineSize=0
       outlineColor=<rgba 0 0 0 1>
      }
     }
     highlightColor {colorUseRule=UseColor_Specified, specifiedColor=<rgba 0 0 0 1>}
     highlightShape {
      bIsDynamicallyLoaded=false
      drawAs=Image
      imageType=NoImage
      mirroring=NoMirror
      resourceName=None
      resourceObject=null
      tiling=NoTile
      imageSize=<32 32>
      margin=<ltrb 0 0 0 0>
      outlineSettings {
       bUseBrushTransparency=false
       roundingType=HalfHeightRadius
       width=0
       color {colorUseRule=UseColor_Specified, specifiedColor=<rgba 0 0 0 0>}
       cornerRadii=<0 0 0 0>
      }
      tintColor {colorUseRule=UseColor_Specified, specifiedColor=<rgba 1 1 1 1>}
      uVRegion {bIsValid=false, max=<0 0>, min=<0 0>}
     }
     selectedBackgroundColor {colorUseRule=UseColor_Foreground, specifiedColor=<rgba 1 0 1 1>}
     shadowColorAndOpacity=<rgba 0 0 0 1>
     shadowOffset=<0 0>
     strikeBrush {
      bIsDynamicallyLoaded=false
      drawAs=Image
      imageType=NoImage
      mirroring=NoMirror
      resourceName=None
      resourceObject=null
      tiling=NoTile
      imageSize=<32 32>
      margin=<ltrb 0 0 0 0>
      outlineSettings {
       bUseBrushTransparency=false
       roundingType=HalfHeightRadius
       width=0
       color {colorUseRule=UseColor_Specified, specifiedColor=<rgba 0 0 0 0>}
       cornerRadii=<0 0 0 0>
      }
      tintColor {colorUseRule=UseColor_Specified, specifiedColor=<rgba 1 1 1 1>}
      uVRegion {bIsValid=false, max=<0 0>, min=<0 0>}
     }
     underlineBrush {
      bIsDynamicallyLoaded=false
      drawAs=Image
      imageType=NoImage
      mirroring=NoMirror
      resourceName=None
      resourceObject=null
      tiling=NoTile
      imageSize=<32 32>
      margin=<ltrb 0 0 0 0>
      outlineSettings {
       bUseBrushTransparency=false
       roundingType=HalfHeightRadius
       width=0
       color {colorUseRule=UseColor_Specified, specifiedColor=<rgba 0 0 0 0>}
       cornerRadii=<0 0 0 0>
      }
      tintColor {colorUseRule=UseColor_Specified, specifiedColor=<rgba 1 1 1 1>}
      uVRegion {bIsValid=false, max=<0 0>, min=<0 0>}
     }
    }
    renderTransform {angle=0, scale=<1 1>, shear=<-12 0>, translation=<0 -6>}
    renderTransformPivot=<0 0.5>
   }
   slot {class=VerticalBoxSlot, properties {padding=<ltrb 0 0 100 0>}}
  }
  RichText_Secondary {
   widgetClassPath=&CommonRichTextBlock
   bIsVariable=true
   parent=VerticalBox_1
   widget=RichText_Secondary
   properties {
    text="+Additional ability info <bs>(Context)</>"
    textStyleSet=&AbilityRichText
    textTransformPolicy=ToUpper
    defaultTextStyleOverride {
     overflowPolicy=Clip
     transformPolicy=None
     colorAndOpacity {colorUseRule=UseColor_Specified, specifiedColor=<rgba 1 1 1 1>}
     font {
      bForceMonospaced=false
      bMaterialIsStencil=false
      fontMaterial=null
      fontObject=&Burbank
      letterSpacing=0
      monospacedWidth=1
      size=20
      skewAmount=0
      typefaceFontName=BurbankRegular-Black
      outlineSettings {
       bApplyOutlineToDropShadows=false
       bMiteredCorners=false
       bSeparateFillAlpha=false
       outlineMaterial=null
       outlineSize=0
       outlineColor=<rgba 0 0 0 1>
      }
     }
     highlightColor {colorUseRule=UseColor_Specified, specifiedColor=<rgba 0 0 0 1>}
     highlightShape {
      bIsDynamicallyLoaded=false
      drawAs=Image
      imageType=NoImage
      mirroring=NoMirror
      resourceName=None
      resourceObject=null
      tiling=NoTile
      imageSize=<32 32>
      margin=<ltrb 0 0 0 0>
      outlineSettings {
       bUseBrushTransparency=false
       roundingType=HalfHeightRadius
       width=0
       color {colorUseRule=UseColor_Specified, specifiedColor=<rgba 0 0 0 0>}
       cornerRadii=<0 0 0 0>
      }
      tintColor {colorUseRule=UseColor_Specified, specifiedColor=<rgba 1 1 1 1>}
      uVRegion {bIsValid=false, max=<0 0>, min=<0 0>}
     }
     selectedBackgroundColor {colorUseRule=UseColor_Foreground, specifiedColor=<rgba 1 0 1 1>}
     shadowColorAndOpacity=<rgba 0 0 0 1>
     shadowOffset=<0 0>
     strikeBrush {
      bIsDynamicallyLoaded=false
      drawAs=Image
      imageType=NoImage
      mirroring=NoMirror
      resourceName=None
      resourceObject=null
      tiling=NoTile
      imageSize=<32 32>
      margin=<ltrb 0 0 0 0>
      outlineSettings {
       bUseBrushTransparency=false
       roundingType=HalfHeightRadius
       width=0
       color {colorUseRule=UseColor_Specified, specifiedColor=<rgba 0 0 0 0>}
       cornerRadii=<0 0 0 0>
      }
      tintColor {colorUseRule=UseColor_Specified, specifiedColor=<rgba 1 1 1 1>}
      uVRegion {bIsValid=false, max=<0 0>, min=<0 0>}
     }
     underlineBrush {
      bIsDynamicallyLoaded=false
      drawAs=Image
      imageType=NoImage
      mirroring=NoMirror
      resourceName=None
      resourceObject=null
      tiling=NoTile
      imageSize=<32 32>
      margin=<ltrb 0 0 0 0>
      outlineSettings {
       bUseBrushTransparency=false
       roundingType=HalfHeightRadius
       width=0
       color {colorUseRule=UseColor_Specified, specifiedColor=<rgba 0 0 0 0>}
       cornerRadii=<0 0 0 0>
      }
      tintColor {colorUseRule=UseColor_Specified, specifiedColor=<rgba 1 1 1 1>}
      uVRegion {bIsValid=false, max=<0 0>, min=<0 0>}
     }
    }
    renderTransform {angle=0, scale=<1 1>, shear=<-8 0>, translation=<0 0>}
    renderTransformPivot=<0 0.5>
   }
   slot {class=VerticalBoxSlot, properties {verticalAlignment=VAlign_Bottom, size {sizeRule=Fill, value=1}}}
  }
 ]
}
