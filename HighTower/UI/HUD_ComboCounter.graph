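# HUD widget that shows a "Hit Combo" damage counter for the owning Fort player pawn.
# On construct it collapses, binds to the pawn's combo increment and reset delegates, waiting for pawn-set if the cast fails, and reads MinHits from game state.
# Updates set the count, randomize its tilt and replay a scale pop.
# Reaching exactly MinHits fades the widget in; ending the combo reverses that Intro animation.
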WidgetBlueprint /HighTower/UI/HUD_ComboCounter
animations [
 UpdateCombo {
  binding_count=2
  length_seconds=0.2566
  object=&self:UpdateCombo
  sequence {bindings [
    Image_40 {
     id=#1
     possessed_class=Image
     tracks [{
       type=MovieScene2DTransformTrack
       sections [{
         type=MovieScene2DTransformSection
         channels [
          Translation.X {type=MovieSceneScriptingFloatChannel}
          Translation.Y {type=MovieSceneScriptingFloatChannel}
          Angle {type=MovieSceneScriptingFloatChannel}
          Scale.X {type=MovieSceneScriptingFloatChannel, keys [[0,0.2517],[2.913,1.5486],[3.7627,1]]}
          Scale.Y {type=MovieSceneScriptingFloatChannel, keys [[0,0.0492],[2.913,1.5486],[3.7627,1]]}
          Shear.X {type=MovieSceneScriptingFloatChannel}
          Shear.Y {type=MovieSceneScriptingFloatChannel}
         ]
        }]
      }]
    }
    Text_Count {
     id=#2
     possessed_class=CommonTextBlock
     tracks [{
       type=MovieScene2DTransformTrack
       sections [{
         type=MovieScene2DTransformSection
         channels [
          Translation.X {type=MovieSceneScriptingFloatChannel}
          Translation.Y {type=MovieSceneScriptingFloatChannel}
          Angle {type=MovieSceneScriptingFloatChannel}
          Scale.X {type=MovieSceneScriptingFloatChannel, keys [[0,0],[2.399,1.4598],[5.1317,1]]}
          Scale.Y {type=MovieSceneScriptingFloatChannel, keys [[0,0],[2.399,1.4598],[5.1317,1]]}
          Shear.X {type=MovieSceneScriptingFloatChannel}
          Shear.Y {type=MovieSceneScriptingFloatChannel}
         ]
        }]
      }]
    }
   ]}
 }
 Intro {
  binding_count=1
  length_seconds=0.2566
  object=&self:Intro
  sequence {bindings [Overlay_15 {
     id=#3
     possessed_class=Overlay
     tracks [{
       type=MovieSceneFloatTrack
       display_name=Render Opacity
       sections [{type=MovieSceneFloatSection, channels [None {type=MovieSceneScriptingFloatChannel, keys [[0,0],[5.1317,1]]}]}]
      }]
    }]}
 }
]
graphs [EventGraph {nodes [
   EventStartCombo() {is=title K2Node_CustomEvent}
   > PlayAnimationForward(InAnimation=Get Intro())
   > SetVisibility(InVisibility=HitTestInvisible)
   EventEndCombo() {is=title K2Node_CustomEvent}
   > PlayAnimationReverse(InAnimation=Get Intro())
   n1 = EventUpdateCombo!() -> (OutputDelegate, then, ComboCount:int, ComboDamage:int) {is=title K2Node_CustomEvent}
   > n2 = Branch(Condition="Equal (Integer)"(A=@n1.ComboCount, B=Get MinHits() -> (MinHits:int=0)) {is=title K2Node_CallFunction})
   > n3 = EventStartCombo() {is=title K2Node_CallFunction}
   n4 = Get Text_Count() -> (Text_Count:Common Text Object Reference)
   "SetText (Text)"(execute=[@n3.then, @n2.else], self=@n4, InText="To Text (Integer)"(Value=@n1.ComboCount))
   > SetRenderTransformAngle!(execute, self:Widget Object Reference=@n4, Angle:float=RandomFloatInRange(Min=-9, Max=3)) -> (then)
   n5 = Get UpdateCombo() -> (UpdateCombo:Widget Animation Object Reference)
   > StopAnimation(InAnimation=@n5)
   > PlayAnimation(InAnimation=@n5)
   n6 = OnPawnSet_Bind() {is=title K2Node_CustomEvent}
   n7 = Cast To FortPlayerPawn(execute=[@n8.then, @n6.then], Object=GetOwningPlayerPawn!(self:User Widget Object Reference) -> (ReturnValue:Pawn Object Reference))
   > Bind Event to On Damage Combo Increment(self:Fort Player Pawn Object Reference=@n7, Delegate=Create Event()) {is=title K2Node_AddDelegate}
   > Bind Event to On Damage Combo Reset(self:Fort Player Pawn Object Reference=@n7, Delegate=Create Event()) {is=title K2Node_AddDelegate}
   > Set MinHits(MinHits:int=GetDamageComboHUDMinHits!(self:Fort Game State Athena Object Reference=GetGameStateAthena!() -> (ReturnValue:Fort Game State Athena Object Reference)) -> (ReturnValue:int=0)) -> (Output_Get:int=0)
   Bind Event to On Pawn Set(
    execute=@n7.CastFailed
    self:Fort HUDContext Object Reference=Get FortHUDContext!() -> (ReturnValue:Fort HUDContext Object Reference) {is=title K2Node_GetBlueprintContextEx}
    Delegate=Create Event()
   ) {is=title K2Node_AddDelegate}
   Event Construct()
   > n8 = SetVisibility(InVisibility=Collapsed)
  ]}]
variables [MinHits {type=integer}]
widgets {
 info {inheritedWidgetCount=1, parentClass=&FortHUDElementWidget, rootWidgetClass=&Overlay, widgetCount=9}
 widgets [
  Overlay_17 {widgetClassPath=&Overlay, widget=Overlay_17}
  Overlay_15 {
   widgetClassPath=&Overlay
   parent=Overlay_17
   widget=Overlay_15
   properties {renderTransform {angle=0, scale=<1 1>, shear=<0 -7>, translation=<-519 -140>}}
   slot {class=OverlaySlot, properties {horizontalAlignment=HAlign_Center, verticalAlignment=VAlign_Center}}
  }
  Image_37 {
   widgetClassPath=&Image
   bIsVariable=true
   parent=Overlay_15
   widget=Image_37
   properties {
    brush {
     bIsDynamicallyLoaded=false
     drawAs=Image
     imageType=NoImage
     mirroring=NoMirror
     resourceName=None
     resourceObject=/Game/Athena/HUD/LTM/Wax/Materials/M_UI_TextGlow_Add
     tiling=NoTile
     imageSize=<265 268>
     margin=<ltrb 0 0 0 0>
     outlineSettings {
      bUseBrushTransparency=false
      roundingType=HalfHeightRadius
      width=0
      color {colorUseRule=UseColor_Specified, specifiedColor=<rgba 0 0 0 0>}
      cornerRadii=<0 0 0 0>
     }
     tintColor {colorUseRule=UseColor_Specified, specifiedColor=<rgba 0.0003 0.2086 0.9734 1>}
     uVRegion {bIsValid=false, max=<0 0>, min=<0 0>}
    }
    renderTransform {angle=0, scale=<1 1>, shear=<0 0>, translation=<13 -17>}
   }
   slot {class=OverlaySlot, properties {horizontalAlignment=HAlign_Center, verticalAlignment=VAlign_Center}}
  }
  Image_39 {
   widgetClassPath=&Image
   bIsVariable=true
   parent=Overlay_15
   widget=Image_39
   properties {
    brush {
     bIsDynamicallyLoaded=false
     drawAs=Image
     imageType=NoImage
     mirroring=NoMirror
     resourceName=None
     resourceObject=/Game/Athena/HUD/Materials/Victory/M_Flare_Spikeball_Rotating2_Inst
     tiling=NoTile
     imageSize=<400 400>
     margin=<ltrb 0 0 0 0>
     outlineSettings {
      bUseBrushTransparency=false
      roundingType=HalfHeightRadius
      width=0
      color {colorUseRule=UseColor_Specified, specifiedColor=<rgba 0 0 0 0>}
      cornerRadii=<0 0 0 0>
     }
     tintColor {colorUseRule=UseColor_Specified, specifiedColor=<rgba 1 1 1 1>}
     uVRegion {bIsValid=false, max=<0 0>, min=<0 0>}
    }
    renderTransform {angle=0, scale=<1 1>, shear=<0 0>, translation=<0 -41>}
   }
   slot {class=OverlaySlot, properties {horizontalAlignment=HAlign_Center, verticalAlignment=VAlign_Center}}
  }
  Image_40 {
   widgetClassPath=&Image
   bIsVariable=true
   parent=Overlay_15
   widget=Image_40
   properties {
    brush {
     bIsDynamicallyLoaded=false
     drawAs=Image
     imageType=NoImage
     mirroring=NoMirror
     resourceName=None
     resourceObject=null
     tiling=NoTile
     imageSize=<256 256>
     margin=<ltrb 0 0 0 0>
     outlineSettings {
      bUseBrushTransparency=false
      roundingType=HalfHeightRadius
      width=0
      color {colorUseRule=UseColor_Specified, specifiedColor=<rgba 0 0 0 0>}
      cornerRadii=<0 0 0 0>
     }
     tintColor {colorUseRule=UseColor_Specified, specifiedColor=<rgba 1 1 1 1>}
     uVRegion {bIsValid=false, max=<0 0>, min=<0 0>}
    }
    renderTransform {angle=123, scale=<1 1>, shear=<0 0>, translation=<0 -26>}
   }
   slot {class=OverlaySlot, properties {horizontalAlignment=HAlign_Center, verticalAlignment=VAlign_Center}}
  }
  Text_Count {
   widgetClassPath=&CommonTextBlock
   bIsVariable=true
   parent=Overlay_15
   widget=Text_Count
   properties {
    text=6
    colorAndOpacity {colorUseRule=UseColor_Specified, specifiedColor=<rgba 1 0 0 1>}
    font {
     bForceMonospaced=false
     bMaterialIsStencil=false
     fontMaterial=null
     fontObject=&Burbank
     letterSpacing=0
     monospacedWidth=1
     size=150
     skewAmount=0
     typefaceFontName=Burbank-Black
     outlineSettings {
      bApplyOutlineToDropShadows=false
      bMiteredCorners=false
      bSeparateFillAlpha=false
      outlineMaterial=null
      outlineSize=2
      outlineColor=<rgba 0 0 0 1>
     }
    }
    renderTransform {angle=3, scale=<1 1>, shear=<-7 0>, translation=<0 0>}
    shadowColorAndOpacity=<rgba 0 0 0 1>
    shadowOffset=<8 8>
   }
   slot {class=OverlaySlot, properties {horizontalAlignment=HAlign_Center, verticalAlignment=VAlign_Center}}
  }
  Image {
   widgetClassPath=&Image
   bIsVariable=true
   parent=Overlay_15
   widget=Image
   properties {
    renderOpacity=0.3238
    brush {
     bIsDynamicallyLoaded=false
     drawAs=Image
     imageType=NoImage
     mirroring=NoMirror
     resourceName=None
     resourceObject=&M_UI_TextGlow_Add
     tiling=NoTile
     imageSize=<265 268>
     margin=<ltrb 0 0 0 0>
     outlineSettings {
      bUseBrushTransparency=false
      roundingType=HalfHeightRadius
      width=0
      color {colorUseRule=UseColor_Specified, specifiedColor=<rgba 0 0 0 0>}
      cornerRadii=<0 0 0 0>
     }
     tintColor {colorUseRule=UseColor_Specified, specifiedColor=<rgba 0.9911 0.0242 0 1>}
     uVRegion {bIsValid=false, max=<0 0>, min=<0 0>}
    }
    renderTransform {angle=0, scale=<1 1>, shear=<0 0>, translation=<-20 34>}
   }
   slot {class=OverlaySlot, properties {horizontalAlignment=HAlign_Center, verticalAlignment=VAlign_Center}}
  }
  Text_Description {
   widgetClassPath=&CommonTextBlock
   bIsVariable=true
   parent=Overlay_15
   widget=Text_Description
   properties {
    lineHeightPercentage=0.7
    minDesiredWidth=1
    text="Hit\r\nCombo"
    textTransformPolicy=ToUpper
    font {
     bForceMonospaced=false
     bMaterialIsStencil=false
     fontMaterial=null
     fontObject=&Burbank
     letterSpacing=0
     monospacedWidth=1
     size=24
     skewAmount=0
     typefaceFontName=BurbankRegular-Black
     outlineSettings {
      bApplyOutlineToDropShadows=false
      bMiteredCorners=false
      bSeparateFillAlpha=false
      outlineMaterial=null
      outlineSize=2
      outlineColor=<rgba 0 0 0 1>
     }
    }
    renderTransform {angle=0, scale=<1 1>, shear=<-12 0>, translation=<117 0>}
    shadowColorAndOpacity=<rgba 0 0 0 1>
    shadowOffset=<4 4>
   }
   slot {class=OverlaySlot, properties {horizontalAlignment=HAlign_Center, verticalAlignment=VAlign_Center}}
  }
  TutorialHighlight {widgetClassPath=&TutorialHighlight}
 ]
}
