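# HUD widget for a single meeple icon that shows an alive or lost state, built on CommonUserWidget.
# SetState resizes the image, then, without animation, tints it white if alive or translucent black if not.
# An animated loss starts a one-second timer calling SubtractMeeple, which plays BreakMeeple, a brief white flash fading to translucent black.
# An animated alive state changes only the size.
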WidgetBlueprint /HighTower/UI/LTM_HUD_Meeple
animations [BreakMeeple {
  binding_count=1
  length_seconds=0.5
  object=&self:BreakMeeple
  sequence {bindings [Image_Meeple {
     id=#1
     possessed_class=Image
     tracks [{
       type=MovieSceneColorTrack
       display_name=Color and Opacity
       sections [{
         type=MovieSceneColorSection
         end_frame=10
         start_frame=0
         channels [
          Color.R {type=MovieSceneScriptingFloatChannel, keys [[0,0],[1,1],[10,0]]}
          Color.G {type=MovieSceneScriptingFloatChannel, keys [[0,0],[1,1],[10,0]]}
          Color.B {type=MovieSceneScriptingFloatChannel, keys [[0,0],[1,1],[10,0]]}
          Color.A {type=MovieSceneScriptingFloatChannel, keys [[0,1],[10,0.2]]}
         ]
        }]
      }]
    }]}
 }]
graphs [
 EventGraph {links [], nodes []}
 SetState {nodes [
   n1 = SetState!() -> (then, Alive:bool=~, Animate:bool=~, IconSize:Vector 2D Structure) {is=title K2Node_FunctionEntry}
   > UpdateSize!(execute, self, inSize:Vector 2D Structure=@n1.IconSize) -> (then)
   > n2 = Branch(Condition=Get Animate() -> (Animate:bool))
   > n3 = Branch(Condition=Get Alive() -> (Alive:bool))
   n4 = Set Timer by Function Name!(
    execute=@n3.else
    Object:Object Reference
    FunctionName:String=SubtractMeeple
    Time:float=1
    bLooping:bool=false
    bMaxOncePerFrame:bool=false
    InitialStartDelay:float=0
    InitialStartDelayVariance:float=0
   ) -> (then, ReturnValue:Timer Handle Structure) {is=title K2Node_CallFunction}
   > Set MeepleSubtractTimer(MeepleSubtractTimer:Timer Handle Structure=@n4) -> (Output_Get:Timer Handle Structure)
   SetColorAndOpacity(
    execute=@n2.else
    self:Image Object Reference=Get Image_Meeple() -> (Image_Meeple:Image Object Reference)
    InColorAndOpacity:Linear Color Structure=Select(Option 0:Linear Color Structure=(R=0,G=0,B=0,A=0.2), Option 1:Linear Color Structure=(R=1,G=1,B=1,A=1), Index=Get Alive()) -> (ReturnValue:Linear Color Structure)
   )
  ]}
 SubtractMeeple {nodes [
   SubtractMeeple() {is=title K2Node_FunctionEntry}
   > PlayAnimation(InAnimation=Get BreakMeeple() -> (BreakMeeple:Widget Animation Object Reference))
   > Clear and Invalidate Timer by Handle(Handle=Get MeepleSubtractTimer() -> (MeepleSubtractTimer:Timer Handle Structure))
  ]}
 UpdateSize {nodes [
   n1 = UpdateSize!() -> (then, inSize:Vector 2D Structure) {is=title K2Node_FunctionEntry}
   > SetDesiredSizeOverride!(execute, self:Image Object Reference=Get Image_Meeple(), DesiredSize:Vector 2D Structure=@n1) -> (then)
  ]}
]
variables [MeepleSubtractTimer {type=TimerHandle}]
widgets {
 info {inheritedWidgetCount=0, parentClass=&CommonUserWidget, rootWidgetClass=&Overlay, widgetCount=4}
 widgets [
  Overlay_1 {widgetClassPath=&Overlay, widget=Overlay_1}
  SizeBox_0 {
   widgetClassPath=&SizeBox
   parent=Overlay_1
   widget=SizeBox_0
   properties {heightOverride=42}
   slot {class=OverlaySlot, properties {horizontalAlignment=HAlign_Fill, verticalAlignment=VAlign_Fill}}
  }
  Overlay_0 {widgetClassPath=&Overlay, parent=SizeBox_0, widget=Overlay_0, slot {class=SizeBoxSlot}}
  Image_Meeple {
   widgetClassPath=&Image
   bIsVariable=true
   parent=Overlay_0
   widget=Image_Meeple
   properties {brush {
     bIsDynamicallyLoaded=false
     drawAs=Image
     imageType=NoImage
     mirroring=NoMirror
     resourceName=None
     resourceObject=null
     tiling=NoTile
     imageSize=<52 60>
     margin=<ltrb 0 0 0 0>
     outlineSettings {
      bUseBrushTransparency=false
      roundingType=HalfHeightRadius
      width=0
      color {colorUseRule=UseColor_Specified, specifiedColor=<rgba 0 0 0 0>}
      cornerRadii=<0 0 0 0>
     }
     tintColor {colorUseRule=UseColor_Specified, specifiedColor=<rgba 1 1 1 1>}
     uVRegion {bIsValid=false, max=<0 0>, min=<0 0>}
    }}
   slot {class=OverlaySlot, properties {horizontalAlignment=HAlign_Fill, verticalAlignment=VAlign_Fill}}
  }
 ]
}
