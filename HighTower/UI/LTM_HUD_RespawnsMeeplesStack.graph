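# HUD widget that shows a respawn count as a stack of meeple icons in a dynamic entry box, cleared on construct.
# Update rebuilds the stack only when Count differs from CachedCount, creating MaxCount LTM_HUD_Meeple entries.
# Entries below Count are alive and the one at index Count animates, with icon size set by bUseLargeMode.
# PreConstruct fills a sample stack at design time.
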WidgetBlueprint /HighTower/UI/LTM_HUD_RespawnsMeeplesStack
graphs [
 EventGraph {nodes [
   Event Construct()
   > Reset!(
    execute
    self:Dynamic Entry Box Object Reference=Get DynamicEntryBox_Meeples() -> (DynamicEntryBox_Meeples:Fort Dynamic Entry Box Object Reference)
    bDeleteWidgets:bool=false
   ) -> (then)
   n1 = Event PreConstruct()
   > Branch(Condition=@n1.IsDesignTime)
   > Update!(execute, self, Count:int=3, MaxCount:int=4) -> (then)
  ]}
 Update {nodes [
   n1 = Update!() -> (then, Count:int, MaxCount:int) {is=title K2Node_FunctionEntry}
   > Branch(Condition="Not Equal (Integer)"!(A:int=@n1, B:int=Get CachedCount() -> (CachedCount:int=0)) -> (ReturnValue:bool) {is=title K2Node_CallFunction})
   > Reset(self=Get DynamicEntryBox_Meeples(), bDeleteWidgets=false)
   > n2 = For Loop(LastIndex=int - int(A=Get MaxCount() -> (MaxCount:int=0), B=1) {is=title K2Node_CallFunction})
   Set CachedCount(execute=@n2.Completed, CachedCount:int=Get Count() -> (Count:int=0)) -> (Output_Get:int=0)
   n3 = Create Entry!(execute=@n2.LoopBody, self:Dynamic Entry Box Object Reference=Get DynamicEntryBox_Meeples()) -> (then, ReturnValue:User Widget Object Reference)
   > n4 = Cast To LTM_HUD_Meeple!(execute, Object=@n3) -> (then, CastFailed, AsLTM HUD Meeple:LTM HUD Meeple Object Reference) {is=title K2Node_DynamicCast}
   n5 = Get Count() -> (Count=0)
   > SetState!(
    execute
    self:LTM HUD Meeple Object Reference=@n4
    Alive:bool="integer > integer"(A=@n5, B=@n2) {is=title K2Node_CallFunction}
    Animate:bool="Equal (Integer)"(A=@n2, B=@n5) {is=title K2Node_CallFunction}
    IconSize:Vector 2D Structure=Select(
     Option 0:Vector 2D Structure=Get SmallSize() -> (SmallSize:Vector 2D Structure)
     Option 1:Vector 2D Structure=Get LargeSize() -> (LargeSize:Vector 2D Structure)
     Index=Get bUseLargeMode() -> (bUseLargeMode:bool)
    ) -> (ReturnValue:Vector 2D Structure)
   ) -> (then)
  ]}
]
variables [
 bUseLargeMode {type=boolean}
 CachedCount {type=integer}
 LargeSize {type=Vector2D, default=<52 60>}
 SmallSize {type=Vector2D, default=<16 20>}
]
widgets {
 info {inheritedWidgetCount=0, parentClass=&CommonUserWidget, rootWidgetClass=/Script/FortniteUI.FortDynamicEntryBox, widgetCount=1}
 widgets [DynamicEntryBox_Meeples {
   widgetClassPath=&FortDynamicEntryBox
   bIsVariable=true
   widget=DynamicEntryBox_Meeples
   properties {entryWidgetClass=&/LTM_HUD_Meeple.LTM_HUD_Meeple_C, numDesignerPreviewEntries=2}
  }]
}
