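# Centered limited-lives HUD widget that shows remaining respawns as a large meeple stack beneath an uppercase caption.
# On a lives count update from the LimitedLivesCounterWidget parent, it hides the caption and passes the count and MaxLives to the meeples stack widget.
# After a short delay it fades the caption in via the ShowText animation, reading "{0} Respawns Left", or "Last Respawn!" when the count is zero or below.
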WidgetBlueprint /HighTower/UI/LTM_HUD_Meeples_Respawn
animations [ShowText {
  binding_count=1
  length_seconds=0.2
  object=&self:ShowText
  sequence {bindings [RespawnsLeft_Text {
     id=#1
     possessed_class=CommonTextBlock
     tracks [{
       type=MovieSceneFloatTrack
       display_name=Render Opacity
       sections [{type=MovieSceneFloatSection, end_frame=4, start_frame=0, channels [None {type=MovieSceneScriptingFloatChannel, keys [[0,0],[4,1]]}]}]
      }]
    }]}
 }]
graphs [
 EventGraph {nodes [
   n1 = Event OnLivesCountUpdated!() -> (OutputDelegate, then, NewCount:int=0) {is=title K2Node_Event}
   > SetRenderOpacity(self=Get RespawnsLeft_Text() -> (RespawnsLeft_Text:Common Text Object Reference), InOpacity=0)
   > Update!(
    execute
    self:LTM HUD Respawns Meeples Stack Object Reference=Get LTM_HUD_RespawnsMeeplesStack() -> (LTM_HUD_RespawnsMeeplesStack:LTM HUD Respawns Meeples Stack Object Reference)
    Count:int=@n1.NewCount
    MaxCount:int=Get MaxLives() -> (MaxLives:int=0)
   ) -> (then)
   > Delay(Duration=1.5)
   > UpdateText!(execute, self, Count:int=@n1.NewCount) -> (then)
  ]}
 UpdateText {nodes [
   n1 = UpdateText!() -> (then, Count:int) {is=title K2Node_FunctionEntry}
   > PlayAnimation(InAnimation=Get ShowText() -> (ShowText:Widget Animation Object Reference))
   > "SetText (Text)"(
    self=Get RespawnsLeft_Text()
    InText=Select(
     Option 0:Text=Format Text!(Format="LOCTEXT(\"{0} Respawns Left\")", 0:int=@n1) -> (Result)
     Option 1:Text="LOCTEXT(\"Last Respawn!\")"
     Index="integer <= integer"!(A:int=@n1, B:int=0) -> (ReturnValue:bool) {is=title K2Node_CallFunction}
    ) -> (ReturnValue:Text)
   )
  ]}
]
variables []
widgets {
 info {inheritedWidgetCount=1, parentClass=/Script/FortniteUI.LimitedLivesCounterWidget, rootWidgetClass=&Overlay, widgetCount=9}
 widgets [
  Overlay_0 {widgetClassPath=&Overlay, widget=Overlay_0}
  SizeBox_2 {
   widgetClassPath=&SizeBox
   parent=Overlay_0
   widget=SizeBox_2
   properties {bOverride_MinDesiredWidth=true, minDesiredHeight=600, minDesiredWidth=600}
   slot {class=OverlaySlot, properties {horizontalAlignment=HAlign_Center, verticalAlignment=VAlign_Center}}
  }
  VerticalBox_0 {
   widgetClassPath=&VerticalBox
   parent=SizeBox_2
   widget=VerticalBox_0
   slot {class=SizeBoxSlot, properties {verticalAlignment=VAlign_Center}}
  }
  Spacer_97 {widgetClassPath=&Spacer, parent=VerticalBox_0, widget=Spacer_97, properties {size=<1 8>}, slot {class=VerticalBoxSlot}}
  RespawnsLeft_Text {
   widgetClassPath=&CommonTextBlock
   bIsVariable=true
   parent=VerticalBox_0
   widget=RespawnsLeft_Text
   properties {
    renderOpacity=0
    textTransformPolicy=ToUpper
    font {
     bForceMonospaced=false
     bMaterialIsStencil=false
     fontMaterial=null
     fontObject=&Burbank
     letterSpacing=0
     monospacedWidth=1
     size=32
     skewAmount=0
     typefaceFontName=BurbankRegular-Black
     outlineSettings {
      bApplyOutlineToDropShadows=false
      bMiteredCorners=false
      bSeparateFillAlpha=false
      outlineMaterial=null
      outlineSize=2
      outlineColor=<rgba 0.0153 0 0.0156 0.5>
     }
    }
    renderTransform {angle=0, scale=<1 1>, shear=<-8 0>, translation=<0 0>}
    shadowColorAndOpacity=<rgba 0.0052 0.0052 0.0056 0.8>
    shadowOffset=<6 4>
   }
   slot {class=VerticalBoxSlot, properties {horizontalAlignment=HAlign_Center, padding=<ltrb 0 16 0 0>}}
  }
  Spacer {widgetClassPath=&Spacer, parent=VerticalBox_0, widget=Spacer, properties {size=<1 8>}, slot {class=VerticalBoxSlot}}
  Overlay_1 {
   widgetClassPath=&Overlay
   parent=VerticalBox_0
   widget=Overlay_1
   slot {class=VerticalBoxSlot, properties {verticalAlignment=VAlign_Center}}
  }
  LTM_HUD_RespawnsMeeplesStack {
   widgetClassPath=&/LTM_HUD_RespawnsMeeplesStack
   bIsVariable=true
   parent=Overlay_1
   widget=LTM_HUD_RespawnsMeeplesStack
   properties {bUseLargeMode=true}
   slot {class=OverlaySlot, properties {horizontalAlignment=HAlign_Center, verticalAlignment=VAlign_Center}}
  }
  TutorialHighlight {widgetClassPath=&TutorialHighlight}
 ]
}
