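# Translucent UI material for a HighTower ability socket, drawing an icon or socket texture with shine, cooldown, charge, whiteout and glow states.
# Scalar switches blend the shined icon into a darkened cooldown with flickering edge flare, then add an optional charge fill.
# Percent-scaled UV masks clip radial gradients colored through curve-atlas rows.
# An offset icon-alpha shadow and pulsing, curve-colored glow combine with icon or socket alpha for opacity.
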Material /HighTower/UI/Materials/M_UI_AbilitySocket
graph {
 properties {blendMode=BLEND_Translucent, materialDomain=MD_UI, numCustomizedUVs=1, pixelDepthOffsetMode=PDOM_Legacy, usedShadingModels=Default Lit}
 nodes [
  n1 = Add(A=TextureCoordinate(), B=AppendVector(A=Constant(), B=ScalarParameter() {group=CooldownEdge, parameterName=CooldownPosition}))
  n2 = VectorParameter() {group=CooldownGradient, parameterName="CenterPosition (V2)"}
  n3 = ScalarParameter() {defaultValue=0.5, group=CooldownGradient, parameterName="Radius (S)"}
  n4 = ScalarParameter() {defaultValue=2.33, group=CooldownGradient, parameterName="Density (S)"}
  n5 = CurveAtlasRowParameter(CurveTime=OneMinus(None=MaterialFunctionCall(UVs=@n1, CenterPosition=@n2.RGB, Radius=@n3, Density=@n4) {materialFunction=&RadialGradientExponential}.RadialGradientExponential)) {
   atlas=/Game/Athena/HUD/Abilities/UIColorAtlas_Ability
   curve=/Game/Athena/HUD/Abilities/UIColorCurve_CooldownCurve
   parameterName=CurveParams
  }
  n6 = Multiply(A=TextureCoordinate(), B=ScalarParameter() {defaultValue=2, group=CooldownGradient, parameterName=CooldownStarScale})
  n7 = AppendVector(A=Constant(), B=ScalarParameter() {group=CooldownGradient, parameterName=StarSpeed})
  Add(A=@n5, B=TextureSample(UVs=Panner(Coordinate=@n6, Speed=@n7)).R)
  n8 = TextureSampleParameter2D() {parameterName=Socket, samplerSource=SSM_Clamp_WorldGroupSettings, texture=&/T_UI_Hightower_Socket_Secondary}
  n9 = TextureSampleParameter2D() {
   parameterName=AbilityIcon
   samplerSource=SSM_Clamp_WorldGroupSettings
   texture=/HighTower/Items/Icons/T-T-Icon-HightowerPickups-Hightower-RepulsorCannon
  }
  n10 = Desaturation(None=@n9.RGB)
  n11 = TextureCoordinate()
  n12 = ScalarParameter() {group=Shine, parameterName="ReferencePlane (S)"}
  n13 = ScalarParameter() {group=Shine, parameterName=shinebumpheight}
  n14 = VectorParameter() {group=Shine, parameterName="CameraVector (V3)"}
  n15 = Multiply(
   A=VectorParameter() {group=Shine, parameterName=ShineColor, defaultValue=<rgba 1 1 1 1>}.A
   B=MaterialFunctionCall!(
    ShineOffset=ScalarParameter() {defaultValue=-1, group=Shine, parameterName=ShineOffset, sliderMax=1.5, sliderMin=-0.5}
    Rotation=ScalarParameter() {defaultValue=0.0476, group=Shine, parameterName=ShineRotation, sliderMax=1}
    RotationCenter=Constant2Vector() {g=1, r=0.5}
    UVs=MaterialFunctionCall!(HeightTexture=@n10, TextureCoordinates=@n11, ReferencePlane=@n12, HeightRatioInput=@n13, CameraVector=@n14.RGB) {materialFunction=/Engine/Functions/Engine_MaterialFunctions02/Utility/BumpOffset_Advanced.BumpOffset_advanced}.UVs
   ) {materialFunction=/Game/Athena/HUD/Materials/Victory/MF_Shine}.Result
  )
  n16 = VectorParameter() {parameterName=CooldownDarken, defaultValue=<rgba 0.008 0.0097 0.0091 0.83>}
  n17 = Constant() {r=1}
  n18 = ScalarParameter() {defaultValue=0.5, group=CooldownEdge, parameterName=CooldownPercent, sliderMax=1}
  n19 = Multiply(
   A=@n5
   B=MaterialFunctionCall!(Texture Scale=AppendVector(A=@n17, B=@n18), UVs=TextureCoordinate(), Offset XY Texturesize Z=Constant3Vector() {constant=<rgba 0 1 1 0>}) {materialFunction=/Engine/Functions/Engine_MaterialFunctions01/Texturing/ScaleUVsAroundPoint}.0-1 mask
  )
  n20 = Constant()
  n21 = Add(
   A=ScalarParameter() {group=CooldownEdge, parameterName=CooldownPosition}
   B=LinearInterpolate(Alpha=ScalarParameter() {defaultValue=0.5, group=CooldownEdge, parameterName=CooldownPercent, sliderMax=1}) {constA=-0.5, constB=0.5}
  )
  n22 = VectorParameter() {group=CooldownEdge, parameterName=FlickerExtents, defaultValue=<rgba 0.5 1 0 1>}
  n23 = Time()
  n24 = ScalarParameter() {defaultValue=1, group=CooldownEdge, parameterName=EdgeFlickerSpeed}
  n25 = LinearInterpolate(
   A=@n22.R
   B=@n22.R
   Alpha=MaterialFunctionCall(Value=Multiply(A=@n23, B=@n24)) {materialFunction=&LinearSine}.Linear Sine
  )
  n26 = VectorParameter() {group=CooldownEdge, parameterName=CooldownEdgeFlareScale, defaultValue=<rgba 1 1 0 1>}
  n27 = MaterialFunctionCall(
   UVs=Add(A=TextureCoordinate(), B=AppendVector(A=@n20, B=@n21))
   Texture Scale=AppendVector(A=Multiply(A=@n25, B=@n26.R), B=Multiply(A=@n25, B=@n26.G))
  ) {materialFunction=&ScaleUVsByCenter}
  n28 = Desaturation(None=Multiply(
    A=TextureSample(UVs=@n27.UVs) {samplerSource=SSM_Clamp_WorldGroupSettings, texture=/Game/Athena/HUD/Materials/Victory/T_UI_StandaloneFlare}.RGB
    B=@n25
   ))
  n29 = LinearInterpolate(
   A=Saturate(None=MaterialFunctionCall!(Blend=@n15, Base=@n9.RGB) {materialFunction=/Engine/Functions/Engine_MaterialFunctions03/Blends/Blend_ColorDodge}.Result)
   B=Add(A=LinearInterpolate(A=@n9.RGB, B=@n16.RGB, Alpha=@n16.RGB), B=Add(A=@n19, B=@n28))
   Alpha=ScalarParameter() {parameterName=CooldownSwitch, sliderMax=1}
  )
  n30 = TextureSampleParameter2D() {parameterName=Socket, texture=&T_UI_Hightower_Socket_Secondary}
  n31 = Add(A=TextureCoordinate(), B=AppendVector(A=Constant(), B=ScalarParameter() {group=Charge, parameterName=ChargePosition}))
  n32 = VectorParameter() {group=Charge, parameterName=ChargeCenterPosition}
  n33 = ScalarParameter() {defaultValue=0.5, group=Charge, parameterName="ChargeRadius (S)"}
  n34 = ScalarParameter() {defaultValue=2.33, group=Charge, parameterName="ChargeDensity (S)"}
  n35 = CurveAtlasRowParameter(CurveTime=OneMinus(None=MaterialFunctionCall(UVs=@n31, CenterPosition=@n32.RGB, Radius=@n33, Density=@n34) {materialFunction=&RadialGradientExponential}.RadialGradientExponential)) {
   atlas=&UIColorAtlas_Ability
   curve=/Game/Athena/HUD/Abilities/UIColorCurve_ChargeCurve
   defaultValue=3
   parameterName=CurveParams
  }
  n36 = Constant() {r=1}
  n37 = ScalarParameter() {defaultValue=0.5, group=Charge, parameterName=ChargePercent, sliderMax=1}
  n38 = Multiply(
   A=Add(A=Ceil(None=@n30.R), B=@n35)
   B=MaterialFunctionCall(Texture Scale=AppendVector(A=@n36, B=@n37), UVs=TextureCoordinate(), Offset XY Texturesize Z=Constant3Vector() {constant=<rgba 0 1 1 0>}) {materialFunction=&ScaleUVsAroundPoint}.0-1 mask
  )
  n39 = Constant()
  n40 = LinearInterpolate(
   A=@n29
   B=Add(A=@n29, B=StaticSwitchParameter(True=@n38, False=@n39) {parameterName=BuiltinCharge})
   Alpha=ScalarParameter() {defaultValue=1, parameterName=ChargeSwitch, sliderMax=1}
  )
  n41 = Constant() {r=1}
  n42 = ScalarParameter() {parameterName=Whiteout}
  n43 = VectorParameter() {group=Shadow, parameterName=ShadowColor, defaultValue=<rgba 0.0052 0.0052 0.0056 0.8>}
  n44 = ScalarParameter() {defaultValue=1, parameterName=UseIcon}
  n45 = LinearInterpolate(A=@n8.RGB, B=LinearInterpolate(A=@n43.RGB, B=LinearInterpolate(A=@n40, B=@n41, Alpha=@n42), Alpha=@n9.A), Alpha=@n44)
  n46 = MaterialFunctionCall(Texture Scale=ScalarParameter() {defaultValue=1.2, group=Glow, parameterName=GlowScale}) {materialFunction=&ScaleUVsByCenter}
  n47 = VectorParameter() {group=Glow, parameterName=GlowRange, defaultValue=<rgba 0.5 13 0 1>}
  n48 = Fmod(A=Multiply(A=Time(), B=ScalarParameter() {defaultValue=0.8, group=Glow, parameterName=PulseSpeed}), B=Constant() {r=1})
  n49 = LinearInterpolate(
   A=Constant()
   B=LinearInterpolate(
    A=@n47.R
    B=@n47.R
    Alpha=MaterialFunctionCall(Value=@n48) {materialFunction=&LinearSine}.Linear Sine
   )
   Alpha=ScalarParameter() {group=Glow, parameterName=GlowToggle}
  )
  n50 = Multiply(
   A=Desaturation(None=TextureSampleParameter2D(UVs=@n46.UVs) {group=Glow, parameterName=GlowTexture, texture=&/T_UI_Hightower_Charge_Secondary}.RGB)
   B=@n49
  )
  n51 = CurveAtlasRowParameter(CurveTime=OneMinus(None=@n50)) {
   atlas=&UIColorAtlas_Ability
   curve=/Game/Athena/HUD/Abilities/UIColorCurve_Ability_Secondary
   defaultValue=3
   parameterName=CurveParams
  }
  n52 = Multiply(A=@n50, B=Saturate(None=@n49))
  n53 = ScalarParameter() {defaultValue=1, group=Glow, parameterName=GlowRamp}
  LinearInterpolate(
   A=LinearInterpolate(A=@n45, B=@n51, Alpha=@n52)
   B=@n45
   Alpha=LinearInterpolate(
    A=MaterialFunctionCall!(In=@n8.A, Ramp Width=@n53) {materialFunction=/Engine/Functions/Engine_MaterialFunctions01/Math/SmoothCeil}.Result
    B=@n9.A
    Alpha=ScalarParameter() {defaultValue=1, parameterName=UseIcon}
   )
  )
  n54 = VectorParameter() {parameterName=GlobalScale, defaultValue=<rgba 1 1 0.5 0.5>}
  MaterialFunctionCall(
   Texture Scale=@n54.R
   UVs=TextureCoordinate()
   Offset XY Texturesize Z=MaterialFunctionCall(R=@n54.B, G=@n54.B, B=Constant() {r=1}) {materialFunction=&AppendMany}.RGB
  ) {materialFunction=&ScaleUVsAroundPoint}
  MaterialFunctionCall(
   R=Constant() {r=0.5}
   G=ScalarParameter() {defaultValue=0.5, group=CooldownEdge, parameterName=CooldownPercent, sliderMax=1}
   B=Constant() {r=1}
  ) {materialFunction=&AppendMany}
  n55 = TextureSampleParameter2D(UVs=Add(A=TextureCoordinate(), B=ScalarParameter() {group=Shadow, parameterName=ShadowOffset})) {
   parameterName=AbilityIcon
   samplerSource=SSM_Clamp_WorldGroupSettings
   texture=&T-T-Icon-HightowerPickups-Hightower-RepulsorCannon
  }
  n56 = Multiply(A=@n55.A, B=@n43.A)
  Max(A=@n52, B=LinearInterpolate(A=@n8.A, B=Max(A=@n9.A, B=@n56), Alpha=@n44))
 ]
}
